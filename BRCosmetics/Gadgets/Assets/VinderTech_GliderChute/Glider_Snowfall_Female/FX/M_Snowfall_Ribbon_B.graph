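# Translucent, unlit, two-sided material for Niagara ribbon and mesh-particle FX on the female Snowfall glider.
# Emissive blends between a panned smoke-trail texture tinted by particle color and a two-layer noise "energy" look with two-color grading.
# Opacity combines view-angle and camera fades, a procedural band and optional curve control.
# World position offset builds from sine/cosine swing, spline width/thickening and dynamic-parameter follow rotation, with a static switch for ribbon mode.
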Material /BRCosmetics/Gadgets/Assets/VinderTech_GliderChute/Glider_Snowfall_Female/FX/M_Snowfall_Ribbon_B
graph {
 properties {
  bUsedWithNiagaraMeshParticles=true
  bUsedWithNiagaraRibbons=true
  blendMode=BLEND_Translucent
  pixelDepthOffsetMode=PDOM_Legacy
  shadingModel=MSM_Unlit
  translucencyLightingMode=TLM_VolumetricPerVertexDirectional
  twoSided=true
  usedShadingModels=Unlit
 }
 nodes [
  n1 = ScalarParameter() {defaultValue=1, parameterName=Texture Tile X}
  n2 = ScalarParameter() {defaultValue=1, parameterName=Texture Tile Y}
  n3 = MaterialFunctionCall(UV Channel=Multiply(A=TextureCoordinate(), B=AppendVector(A=@n1, B=@n2))) {materialFunction=&LinearGradient}
  n4 = MaterialFunctionCall(X=Add(A=VertexColor().A, B=@n3.UGradient), Y=@n3.VGradient) {materialFunction=&MakeFloat2}
  n5 = AppendVector(A=ScalarParameter() {parameterName=texture Speed X}, B=ScalarParameter() {parameterName=texture Speed Y})
  n6 = Desaturation(
   None=TextureSampleParameter2D(UVs=Panner(Coordinate=@n4.Result, Speed=@n5) {bFractionalPart=true, speedX=-0.5}) {parameterName=Texture, texture=/Game/Effects/Fort_Effects/Textures/Smoke/T_Smoke_Trail_Soft_2}.RGB
   Fraction=ScalarParameter() {parameterName=Desaturation Amt}
  )
  n7 = ParticleColor()
  n8 = Multiply(A=@n6, B=@n7.A)
  n9 = ScalarParameter() {defaultValue=1, parameterName=opacity - look at}
  n10 = ScalarParameter() {defaultValue=1, parameterName=opacity - look away}
  n11 = CameraPositionWS()
  n12 = VectorParameter() {parameterName=Actor/position Offset, channelNames {a=Is Local, b=Z, g=Y, r=X}, defaultValue=<rgba 0 0 0 1>}
  n13 = Add(
   A=LinearInterpolate(A=ActorPositionWS(), B=ObjectPositionWS(), Alpha=ScalarParameter() {defaultValue=1, parameterName=Actor vs obcet position})
   B=LinearInterpolate(A=@n12, B=Transform(None=@n12) {transformSourceType=TRANSFORMSOURCE_Local}, Alpha=@n12)
  )
  n14 = VectorParameter() {parameterName=Object vector to comapre, defaultValue=<rgba 0 1 0 1>}
  n15 = DotProduct(
   A=Normalize(VectorInput=Subtract(A=@n11, B=@n13))
   B=Normalize(VectorInput=Transform(None=@n14.RGB) {transformSourceType=TRANSFORMSOURCE_Local})
  )
  n16 = OneMinus(None=LinearInterpolate(A=ConstantBiasScale(None=@n15), B=Abs(None=@n15), Alpha=ScalarParameter() {parameterName=fade both sides}))
  n17 = VectorParameter() {parameterName=Dot Options, channelNames {a=pre mul, b=add, g=mul, r=pow}, defaultValue=<rgba 1 1 0 1>}
  n18 = Power(Base=Clamp(None=Multiply(A=@n16, B=@n17.pre mul)), Exp=@n17.pow) {constExponent=1}
  n19 = Clamp(None=Add(A=Multiply(A=@n18, B=@n17.mul), B=@n17.add) {constB=0})
  n20 = Multiply(
   A=MaterialFunctionCall(Fade Length=ScalarParameter() {parameterName=Cam Fade length}, Fade Offset=ScalarParameter() {parameterName=Cam Fade offset}) {materialFunction=&CameraDepthFade}.Result
   B=Clamp(None=LinearInterpolate(A=@n9, B=@n10, Alpha=@n19))
  )
  n21 = ScalarParameter() {defaultValue=0.5, parameterName=band width}
  n22 = ScalarParameter() {parameterName=Band Sharpness}
  n23 = StaticBool() {value=true}
  n24 = Power(Base=TextureCoordinate(), Exp=ScalarParameter() {defaultValue=2, parameterName=BandExp})
  n25 = Multiply(
   A=Multiply(A=NamedRerouteUsage(), B=Multiply(A=@n8, B=@n20))
   B=Multiply(A=MaterialFunctionCall!(Width=@n21, Sharpness=@n22, Offset, Direction Switch=@n23, Compare, Input Coordinates=@n24) {materialFunction=/Engine/Functions/Engine_MaterialFunctions03/Procedurals/GeneratedBand}.Result)
  )
  n26 = NamedRerouteUsage()
  LinearInterpolate(
   A=@n25
   B=Multiply(
    A=MaterialFunctionCall(Float3=@n26) {materialFunction=&BreakOutFloat3Components}.B
    B=@n25
   )
   Alpha=ScalarParameter() {parameterName=use param curve opac}
  )
  n27 = Desaturation(
   None=Multiply(
    A=VectorParameter() {parameterName=Secondary Color, defaultValue=<rgba 1 1 1 1>}.RGB
    B=ScalarParameter() {defaultValue=1, group=Color, parameterName=Secondary Color Intensity}
   )
   Fraction=ScalarParameter() {defaultValue=0.2, group=Color, parameterName=Secondary Color Desaturation}
  )
  n28 = TextureCoordinate()
  n29 = AppendVector(
   A=ScalarParameter() {defaultValue=1.5, group=Noise, parameterName=NoiseTexture_Tile_U}
   B=ScalarParameter() {defaultValue=0.5, group=Noise, parameterName=NoiseTexture_Tile_V}
  )
  n30 = ScalarParameter() {group=Noise, parameterName=NoiseTexture_Pan_U}
  n31 = ScalarParameter() {group=Noise, parameterName=NoiseTexture_Pan_V}
  n32 = TextureSampleParameter2D(UVs=Panner(Coordinate=Multiply(A=@n28, B=@n29), Speed=AppendVector(A=@n30, B=@n31)) {speedY=0.3}) {group=Noise, parameterName=NoiseTexture, texture=/Game/Packages/Fortress_Transylvania/Haunted_Forest/Swamps/Textures/2ndInfectedMapLinear2}
  n33 = TextureCoordinate()
  n34 = AppendVector(
   A=ScalarParameter() {defaultValue=1.5, group=Noise, parameterName=NoiseTexture_Tile_U 2}
   B=ScalarParameter() {defaultValue=0.5, group=Noise, parameterName=NoiseTexture_Tile_V 2}
  )
  n35 = ScalarParameter() {group=Noise, parameterName=NoiseTexture_Pan_U 2}
  n36 = ScalarParameter() {group=Noise, parameterName=NoiseTexture_Pan_V 2}
  n37 = TextureSampleParameter2D(UVs=Panner(Coordinate=Multiply(A=@n33, B=@n34), Speed=AppendVector(A=@n35, B=@n36)) {speedY=0.3}) {
   group=Noise
   parameterName=NoiseTexture 2
   samplerType=SAMPLERTYPE_LinearGrayscale
   texture=/BRCosmetics/Weapons/FORT_Melee/Pickaxe_SnowFall_Female/FX/T_SnowFall_pick_basecolor
  }
  n38 = MaterialFunctionCall(UV Channel=LinearInterpolate(A=TextureCoordinate(), B=TextureCoordinate() {coordinateIndex=1}, Alpha=ScalarParameter() {parameterName=noise transition Use UV1})) {materialFunction=&LinearGradient}
  n39 = ScalarParameter() {defaultValue=1, parameterName="noise transition Use V Gradient "}
  n40 = Add(
   A=OneMinus(None=LinearInterpolate(A=@n38.UGradient, B=@n38.UGradient, Alpha=@n39))
   B=ScalarParameter() {defaultValue=-0.2, group=Noise, parameterName=noise Transition gradient Offset}
  )
  n41 = ScalarParameter() {defaultValue=3, group=Noise, parameterName=noise Transition gradient Power}
  n42 = Saturate(None=Multiply(A=Power(Base=@n40, Exp=@n41), B=ScalarParameter() {defaultValue=5, group=Noise, parameterName=noise Transition gradient Intencity}))
  n43 = Multiply(
   A=Power(
    Base=LinearInterpolate(A=@n32.R, B=@n37.R, Alpha=@n42)
    Exp=ScalarParameter() {defaultValue=1, group=Noise, parameterName=NoiseTexture_Power}
   )
   B=ScalarParameter() {defaultValue=1, group=Noise, parameterName=NoiseTexture_Intensity}
  )
  n44 = Multiply(A=@n27, B=@n43)
  n45 = Multiply(
   A=VectorParameter() {parameterName=Primary Color, defaultValue=<rgba 1 1 1 0>}.RGB
   B=ScalarParameter() {defaultValue=1, group=Color, parameterName=Primary Color Intensity}
  )
  n46 = ScalarParameter() {defaultValue=0.15, group=Color, parameterName=Primary Color Desaturation}
  n47 = CurveAtlasRowParameter(CurveTime=ComponentMask(None=TextureCoordinate()) {r=true}) {
   atlas=/Game/Athena/LinearColorCurves/TempAthenaCurveAtlas
   curve=/Game/Weapons/FORT_Melee/Pickaxe_Elastic/Fire/Curve_Pickaxe_Elastic_Fire_01
   defaultValue=14
  }
  n48 = MaterialFunctionCall(Float2=LinearInterpolate(A=Constant3Vector() {constant=<rgba 1 1 1 0>}, B=@n47, Alpha=ScalarParameter() {defaultValue=1, parameterName=Use Param})) {materialFunction=&BreakOutFloat2Components}
  n49 = LinearInterpolate(A=@n48.R, B=Saturate(None=Multiply(A=@n48.R) {constB=10}), Alpha=ScalarParameter() {group=Color, parameterName=Use_clear_boundaries})
  n50 = Saturate(None=@n43)
  n51 = ScalarParameter() {parameterName=use noise as color lerp}
  n52 = Multiply(A=LinearInterpolate(A=@n27, B=Desaturation(None=@n45, Fraction=@n46), Alpha=LinearInterpolate(A=@n49, B=@n50, Alpha=@n51)), B=@n48.G)
  n53 = LinearInterpolate(A=@n52, B=Multiply(A=@n52, B=@n43), Alpha=ScalarParameter() {defaultValue=1, parameterName=Multiply mask in color})
  n54 = Clamp(None=TwoSidedSign())
  n55 = LinearInterpolate(
   A=ScalarParameter() {parameterName=particle color mul Low}
   B=ScalarParameter() {defaultValue=1, parameterName=particle color mul High}
   Alpha=ComponentMask(None=LinearInterpolate(A=@n44, B=@n53, Alpha=@n54)) {r=true}
  )
  n56 = Multiply(A=@n55, B=ParticleColor().RGB)
  n57 = ScalarParameter() {parameterName=Use_ParticleColor}
  MaterialFunctionCall(Emissive=LinearInterpolate(
    A=Multiply(A=@n6, B=@n7.RGB)
    B=LinearInterpolate(A=@n55, B=@n56, Alpha=@n57)
    Alpha=ScalarParameter() {parameterName=use Energy From Snowfall}
   )) {materialFunction=&MF_FX_EmissiveAdjustments}
  n58 = NamedRerouteUsage()
  n59 = NamedRerouteUsage()
  n60 = Add(A=NamedRerouteUsage(), B=@n58)
  Multiply(
   A=StaticSwitchParameter(True=@n58, False=Add(A=@n59, B=@n60)) {group=********, parameterName=ChangeToRibbonParticleMode}
   B=ScalarParameter() {defaultValue=1, group=**Debug, parameterName=TotalWPOScale}
  )
  n61 = TextureCoordinate()
  n62 = MaterialFunctionCall(Float2=@n61) {materialFunction=&BreakOutFloat2Components}
  n63 = OneMinus(None=@n62.R)
  n64 = ScalarParameter() {group=01.GlobalUVSettings, parameterName=FlipU}
  n65 = OneMinus(None=@n62.G)
  n66 = ScalarParameter() {group=01.GlobalUVSettings, parameterName=FlipV}
  NamedRerouteDeclaration(None=MaterialFunctionCall(X=LinearInterpolate(A=@n62.R, B=@n63, Alpha=@n64), Y=LinearInterpolate(A=@n62.G, B=@n65, Alpha=@n66)) {materialFunction=&MakeFloat2}.Result) {name=R_UV, nodeColor=<rgba 0 1 0.2266 1>}
  n67 = DynamicParameter() {parameterIndex=3, defaultValue=<rgba 0 0 0 0>, paramNames [Timer,MeshScale,NotUse,NotUse]}
  n68 = LinearInterpolate(A=Time() {bOverride_Period=true, period=360}, B=@n67.Timer, Alpha=Saturate(None=Ceil(None=@n67.Timer)))
  n69 = Add(A=@n68, B=Multiply(A=VertexColor().A, B=ScalarParameter() {defaultValue=10, group=02.GlobalTimeSetting, parameterName=RandomScale}))
  NamedRerouteDeclaration(None=StaticSwitchParameter(True=@n68, False=StaticSwitchParameter(True=@n68, False=@n69) {group=********, parameterName=DisableThriveSpecialColorFixing}) {group=********, parameterName=ChangeToRibbonParticleMode}) {name=R_TimeRandom, nodeColor=<rgba 0.7812 1 0 1>}
  n70 = CurveAtlasRowParameter(CurveTime=MaterialFunctionCall(Float2=NamedRerouteUsage()) {materialFunction=&BreakOutFloat2Components}.R) {
   atlas=&/CA_SnowfallSlplineCurves
   curve=&/CC_SnowfallMeshRibbon
   parameterName=Ribbon Param Curve
  }
  NamedRerouteDeclaration(None=AppendVector(A=@n70, B=@n70)) {name=R_ParamCurve, nodeColor=<rgba 0.6172 1 0 1>}
  n71 = VectorParameter() {group=40.NiagaraAnimation, parameterName=GrowAnimation_Detail, defaultValue=<rgba 1.5 1 -1.5 0>}
  n72 = Multiply(
   A=MaterialFunctionCall(Float2=NamedRerouteUsage()) {materialFunction=&BreakOutFloat2Components}.R
   B=@n71.R
  )
  n73 = DynamicParameter() {
   parameterIndex=2
   defaultValue=<rgba 1 0 0 1>
   paramNames ["MeshPosZ(OnlyOnSplineMesh)", "EnableFollow(OnlyOnSplineMesh)", NoiseReducer, Grow]
  }
  n74 = Saturate(None=@n73.Grow)
  n75 = LinearInterpolate(A=@n71.G, B=@n71.G, Alpha=Add(A=@n71.A, B=OneMinus(None=@n74)))
  NamedRerouteDeclaration(None=Saturate(None=Add(A=@n72, B=@n75))) {name=R_NiagaraAnimation_Grow_SplineWidthScale, nodeColor=<rgba 0.7109 1 0 1>}
  n76 = DynamicParameter() {parameterIndex=3, defaultValue=<rgba 0 0 0 0>, paramNames [Timer,MeshScale,NotUse,NotUse]}
  n77 = NamedRerouteUsage()
  n78 = Multiply(A=ScalarParameter() {group=32.SplineWPO, parameterName=Spline_WidthBase}, B=@n77)
  n79 = Multiply(
   A=MaterialFunctionCall(Float4=NamedRerouteUsage()) {materialFunction=&BreakOutFloat4Components}.R
   B=@n78
  )
  n80 = ScalarParameter() {group=32.SplineWPO, parameterName=Spline_SplineThicken_WidthTip}
  n81 = NamedRerouteUsage()
  n82 = WorldPosition() {worldPositionShaderOffset=WPT_ExcludeAllShaderOffsets}
  n83 = Constant3Vector() {constant=<rgba 0 1 0 0>}
  n84 = Transform(None=Multiply(A=Normalize(VectorInput=@n83), B=ScalarParameter() {defaultValue=1, group=32.SplineWPO, parameterName=Spline_CurveControl_FlipVDir}))
  n85 = NamedRerouteUsage()
  n86 = Multiply(A=Subtract(A=MaterialFunctionCall(Float2=@n85) {materialFunction=&BreakOutFloat2Components}.G) {constB=0.5}) {constB=2}
  n87 = StaticSwitchParameter(
   True=MaterialFunctionCall!(
    UseRemapCurver=ScalarParameter() {group=32.SplineWPO, parameterName=Spline_SplineThicken_UseCurverForSplineThickenWidth}
    WidthBase=@n78
    WidthRemapCurver=@n79
    TangentUToWorld=NamedRerouteUsage()
    TangentVToWorld=NamedRerouteUsage()
    WidthTip=Multiply(A=@n80, B=@n77)
    WorldPosition=Add(A=@n81, B=@n82.XYZ)
    UVs for Projection
    UVs for Thickness
    UVs for texturing
    Expand U or V UV Channel=StaticBool()
    DeriveNormalZ
    AdditionalNormal
    AngleCorrectedNormals
    FlattenPixelNormal
    UVs For NormalShading
   ) {materialFunction=/Game/Gadgets/Assets/VinderTech_GliderChute/Glider_Thrive_Female/FX/Material/MF_Thrive_CustomSplineThicken}.WorldPositionOffset
   False=Multiply(A=Multiply(A=@n84, B=@n86), B=@n79)
  ) {group=32.SplineWPO, parameterName=UseSplineThickness}
  NamedRerouteDeclaration(None=LinearInterpolate(A=@n87, B=Multiply(A=@n87, B=@n76.MeshScale), Alpha=ScalarParameter() {parameterName=ApplyDynamicInputMeshScale})) {name=R_SplineWPO, nodeColor=<rgba 0.5703 0 1 1>}
  n88 = Constant() {r=1}
  n89 = ScalarParameter() {parameterName=depthFadeDistance}
  n90 = Constant() {r=1}
  NamedRerouteDeclaration(None=ShadingPathSwitch(Default=DepthFade(Opacity=@n88, FadeDistance=@n89), Forward=@n90, Mobile=@n90)) {name=R_CustomDepthCullAndFade, nodeColor=<rgba 0 0.6953 1 1>}
  NamedRerouteDeclaration(None=@n61) {name=R_OriginalUV, nodeColor=<rgba 0 1 0.4869 1>}
  n91 = DynamicParameter() {parameterIndex=3, defaultValue=<rgba 0 0 0 0>, paramNames [Timer,MeshScale,NotUse,NotUse]}
  n92 = VectorParameter() {group=30.SwingWPO, parameterName=Swing_Cosine_Mask, channelNames {a="", b="", g=offset, r=Multiply}, defaultValue=<rgba 2 -1.1 0 1>}
  n93 = MaterialFunctionCall(Float2=NamedRerouteUsage()) {materialFunction=&BreakOutFloat2Components}
  n94 = Add(A=Multiply(A=@n92.Multiply, B=@n93.R), B=@n92.offset)
  n95 = VectorParameter() {group=30.SwingWPO, parameterName=Swing_Cosine_Property, channelNames {a="", b=Frequency, g=Strength, r=Speed}, defaultValue=<rgba 1.5 5 2 1>}
  n96 = NamedRerouteUsage()
  n97 = Cosine(None=Add(A=Multiply(A=@n93.R, B=@n95.Frequency), B=Multiply(A=@n96, B=@n95.Speed)))
  n98 = VectorParameter() {group=30.SwingWPO, parameterName=Swing_Cosine_Axis, defaultValue=<rgba 0 1 0 0>}
  n99 = Multiply(A=OneMinus(None=Saturate(None=@n94)), B=Multiply(A=Multiply(A=@n97, B=@n95.Strength), B=Transform(None=@n98.RGB)))
  n100 = VectorParameter() {group=30.SwingWPO, parameterName=Swing_Sine_Mask, channelNames {a="", b="", g=Offset, r=Multiply}, defaultValue=<rgba 2 -0.375 0 1>}
  n101 = Multiply(A=@n100.Multiply, B=@n93.R)
  n102 = OneMinus(None=Saturate(None=Add(A=@n101, B=@n100.Offset)))
  n103 = VectorParameter() {group=30.SwingWPO, parameterName=Swing_Sine_Property, channelNames {a="", b=Frequency, g=Strength, r=Speed}, defaultValue=<rgba 1.5 20 2 1>}
  n104 = Add(A=Multiply(A=@n93.R, B=@n103.Frequency), B=Multiply(A=@n96, B=@n103.Speed))
  n105 = Multiply(
   A=Multiply(A=Sine(None=@n104), B=@n103.Strength)
   B=Transform(None=VectorParameter() {group=30.SwingWPO, parameterName=Swing_Sine_Axis, defaultValue=<rgba 0 0 1 1>}.RGB)
  )
  n106 = StaticSwitchParameter(True=@n99, False=Add(A=Multiply(A=@n102, B=@n105), B=@n99)) {group=********, parameterName=ChangeToRibbonParticleMode}
  NamedRerouteDeclaration(None=LinearInterpolate(A=@n106, B=Multiply(A=@n106, B=@n91.MeshScale), Alpha=ScalarParameter() {parameterName=ApplyDynamicInputMeshScale})) {name=R_SwingWPO, nodeColor=<rgba 0.6172 1 0 1>}
  n107 = DynamicParameter() {paramNames ["PosX(OnlyOnSplineMesh)", "PosY(OnlyOnSplineMesh)", "PosZ(OnlyOnSplineMesh)", "RestX(OnlyOnSplineMesh)"]}
  n108 = DynamicParameter() {
   parameterIndex=1
   paramNames ["RestY(OnlyOnSplineMesh)", "RestZ(OnlyOnSplineMesh)", "MeshPosX(OnlyOnSplineMesh)", "MeshPosY(OnlyOnSplineMesh)"]
  }
  n109 = Normalize(VectorInput=AppendVector(A=AppendVector(A=@n107."RestX(OnlyOnSplineMesh)", B=@n108."RestY(OnlyOnSplineMesh)"), B=@n108."RestZ(OnlyOnSplineMesh)"))
  n110 = AppendVector(A=AppendVector(A=@n107."PosX(OnlyOnSplineMesh)", B=@n107."PosX(OnlyOnSplineMesh)"), B=@n107."PosZ(OnlyOnSplineMesh)")
  n111 = Constant3Vector() {constant=<rgba 0 0 1 0>}
  n112 = DynamicParameter() {
   parameterIndex=2
   defaultValue=<rgba 1 0 0 1>
   paramNames ["MeshPosZ(OnlyOnSplineMesh)", "EnableFollow(OnlyOnSplineMesh)", NoiseReducer, Grow]
  }
  n113 = AppendVector(A=AppendVector(A=@n108."MeshPosX(OnlyOnSplineMesh)", B=@n108."MeshPosX(OnlyOnSplineMesh)"), B=@n112."MeshPosZ(OnlyOnSplineMesh)")
  n114 = Normalize(VectorInput=Subtract(A=Add(A=@n110, B=@n111), B=@n113))
  n115 = CrossProduct(A=@n109, B=@n114)
  n116 = DotProduct(A=@n114, B=@n109)
  n117 = MaterialFunctionCall(Float2=NamedRerouteUsage()) {materialFunction=&BreakOutFloat2Components}
  n118 = ScalarParameter() {group="31.FollowWPO(NiagaraDeform)", parameterName=FollowWPO_Offset}
  n119 = Multiply(A=Add(A=OneMinus(None=@n117.R), B=@n118), B=Divide(B=Add(A=@n118)))
  n120 = Multiply(
   A=@n119
   B=Multiply(
    A=ArccosineFast!(None=@n116)
    B=ScalarParameter() {defaultValue=0.1592, group="31.FollowWPO(NiagaraDeform)", parameterName=FollowWPO_Rigidity}
   )
  )
  n121 = Add(A=NamedRerouteUsage(), B=WorldPosition() {worldPositionShaderOffset=WPT_ExcludeAllShaderOffsets}.XYZ)
  n122 = DynamicParameter() {
   parameterIndex=2
   defaultValue=<rgba 1 0 0 1>
   paramNames ["MeshPosZ(OnlyOnSplineMesh)", "EnableFollow(OnlyOnSplineMesh)", NoiseReducer, Grow]
  }
  NamedRerouteDeclaration(None=LinearInterpolate(
    B=RotateAboutAxis(NormalizedRotationAxis=@n115, RotationAngle=@n120, PivotPoint=@n113, Position=@n121)
    Alpha=@n122."EnableFollow(OnlyOnSplineMesh)"
   )) {name=R_NiagaraDeformationWPO, nodeColor=<rgba 0 1 0.1797 1>}
  n123 = Transform(None=Constant3Vector() {constant=<rgba 1 0 0 0>})
  NamedRerouteDeclaration(None=LinearInterpolate(
    A=@n123
    B=MaterialFunctionCall!(VectorToRotate=@n123, LookAtVector=@n114, RestingVector=@n109) {materialFunction=/Engine/Functions/Engine_MaterialFunctions01/Vectors/RotateVector}.Result
    Alpha=@n119
   )) {name=R_NiagaraDeformation_UDirWS, nodeColor=<rgba 0 1 0.4844 1>}
  n124 = Transform(None=Constant3Vector() {constant=<rgba 1 1 0 0>})
  NamedRerouteDeclaration(None=LinearInterpolate(
    A=@n124
    B=MaterialFunctionCall(VectorToRotate=@n124, LookAtVector=@n114, RestingVector=@n109) {materialFunction=&RotateVector}.Result
    Alpha=@n119
   )) {name=R_NiagaraDeformation_VDirWS, nodeColor=<rgba 0 1 0.4844 1>}
 ]
}
