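# Niagara effect for the Snowfall (female) VinderTech glider, layering a persistent mesh effect with pale blue cloud sprites.
# A long-lived particle with four mesh renderers follows the owner, passing a lagged target point, axis and growth timer to materials through dynamic parameters.
# Cloud emitters spawn on skeletal mesh bones with curl noise, split into gameplay and front-end variants.
# Chunk, clump and energy ribbon emitters are present but disabled.
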NiagaraSystem /BRCosmetics/Gadgets/Assets/VinderTech_GliderChute/Glider_Snowfall_Female/FX/NS_snowfall_Glider
dependencies {
 usedDataInterfaces [
  &NiagaraDataInterfacePlatformSet
  &NiagaraDataInterfaceCamera
  &NiagaraDataInterfaceMeshRendererInfo
  &NiagaraDataInterfaceCurve
  &NiagaraDataInterfaceVectorCurve
  &NiagaraDataInterfaceArrayFloat3
  &NiagaraDataInterfaceColorCurve
  &NiagaraDataInterfaceVector2DCurve
 ]
 usedDynamicInputs []
 usedModules [&SystemState, &Timeline]
}
emitters {
 Chunks001 {
  bEnabled=false
  emitterUpdateScript [
   &EmitterState()
   &SpawnRate(SpawnRate=50, Use Spawn Probability=false)
  ]
  particleSpawnScript [
   /SetVariables(Particles.Intensity:NiagaraFloat=&FloatFromCurve(
     FloatCurve=DataInterface() {propertyValues {Curve {keys [<at 0 0>,<at 0.8 0>,<at 1 1>]}}}
     CurveIndex=&RandomRangeFloat(Minimum=0, Maximum=1)
    )) {isSetParameters=true}
   &InitializeParticle(
    Color Maximum=<rgba 0.2362 0.3943 0.5906 1>
    Color Minimum=<rgba 1 1 1 1>
    Color Mode=::NewEnumerator2
    Lifetime=&Multiply_Float(
     A=&RandomRangeFloat(Minimum=0.3, Maximum=0.4)
     B=&Lerp_Float(A=0.6, B=1, Alpha=$Particles.Intensity)
    )
    Mass Max=2
    Mass Mode=::NewEnumerator3
    Mesh Scale=&Multiply_VectorByFloat(Vector=&RandomRangeVector(Minimum=<2 0.75 0.5>, Maximum=<4 1.25 2>), Float=0.05)
    Mesh Scale Mode=::NewEnumerator4
    Sprite Rotation Mode=::NewEnumerator2
    UsePositionOffset=false
    Write Material Random=true
   )
   &FN_SetSpawnGroup(Number of Spawn Groups=2)
   &SkeletalMeshLocation(
    Bone / Socket Index=7
    Bone Sampling Mode=::NewEnumerator4
    Interpolate Mesh Sampling=false
    MaskSpawnGroup=false
    Orientation Sampling=::NewEnumerator1
    Sampled Position Offset=&Add_Vector(
     A=&Lerp_Vector(
      A=<32 35 0>
      B=<-32 35 0>
      Alpha=/Niagara/DynamicInputs/Integer/V2/Round_Float(Float:NiagaraFloat=&RandomRangeFloat(Minimum=0, Maximum=1))
     )
     B=<0 0 0>
    )
    Skeletal Mesh=$User.SkeletalMesh
   )
   &SphereLocation(
    MaskSpawnGroup=false
    Override Seed=false
    Sphere Radius=10
    Surface Only=false
    UseNonUniformScale=false
    UseOffset=false
    UseSphereOrientation=false
   ) {enabled=false}
   &AddVelocityInCone(
    Cone Angle=40
    Cone Axis=<0 1 0>
    Module.Use Velocity Falloff On Cone Axis=true
    Velocity Distribution Along Cone Axis=0.75
    Velocity Falloff Away From Cone Axis=0.333
    Velocity Strength=&Multiply_Float(A=&RandomRangeFloat(Minimum=180, Maximum=200), B=4)
   )
   &AddVelocity(Velocity=&RandomRangeVector(Minimum=<0 0 40>, Maximum=<0 0 60>))
  ]
  particleUpdateScript [
   &ParticleState()
   &GravityForce(Gravity=<0 0 -250>)
   &InheritSourceMovement(ApplyForce=false, ApplyPosition=true, ApplyPositionFalloff=false) {enabled=false}
   &Drag(Drag=&RandomRangeFloat(Minimum=0.8, Maximum=1.2), Use Linear Drag=true, Use Rotational Drag=false)
   &SolveForcesAndVelocity(Clamp Velocity=false, Limit Acceleration=false)
   &ScaleSpriteSize(
    Uniform Curve Scale=4
    Uniform Curve Sprite Scale=DataInterface() {propertyValues {Curve {keys [
        {arriveTangent=9.0725, interpMode=RCIM_Cubic, leaveTangent=9.0725, tangentMode=RCTM_User, time=0, value=0}
        {arriveTangent=-0.0529, interpMode=RCIM_Cubic, leaveTangent=-0.0529, tangentMode=RCTM_User, time=0.3, value=0.9847}
        {arriveTangent=-1.377, interpMode=RCIM_Cubic, leaveTangent=-1.377, tangentMode=RCTM_User, time=0.8, value=0.65}
        {arriveTangent=-0.0405, interpMode=RCIM_Cubic, leaveTangent=-0.0405, tangentMode=RCTM_User, time=1, value=0}
       ]}}}
   )
   &ScaleMeshSize(Scale Factor=&Multiply_VectorByFloat(
     Vector=<1 1 1>
     Float=&FloatFromCurve(FloatCurve=DataInterface() {propertyValues {Curve {keys [
          {arriveTangent=0.6971, interpMode=RCIM_Cubic, leaveTangent=0.6971, tangentMode=RCTM_User, time=0, value=0}
          {arriveTangent=-0.0529, interpMode=RCIM_Cubic, leaveTangent=-0.0529, tangentMode=RCTM_User, time=0.2, value=1.2}
          {arriveTangent=-1.6278, interpMode=RCIM_Cubic, leaveTangent=-1.6278, tangentMode=RCTM_User, time=0.8, value=0.65}
          {arriveTangent=-0.0405, interpMode=RCIM_Cubic, leaveTangent=-0.0405, tangentMode=RCTM_User, time=1, value=0}
         ]}}})
    ))
  ]
  rendererClasses [&NiagaraMeshRendererProperties]
  renderers [{rendererClass=&NiagaraMeshRendererProperties, rendererIndex=0}]
 }
 Clumps001 {
  bEnabled=false
  emitterUpdateScript [
   &EmitterState()
   &SpawnRate(SpawnRate=80, Use Spawn Probability=false)
  ]
  particleSpawnScript [
   /SetVariables(Particles.InitSpriteRotRate=&RandomRangeFloat(Minimum=-2, Maximum=2)) {isSetParameters=true}
   &InitializeParticle(
    Color Minimum=<rgba 0.3755 0.822 0.9387 1>
    Color Mode=::NewEnumerator2
    Lifetime Max=0.7
    Lifetime Min=0.4
    Lifetime Mode=::NewEnumerator1
    Sprite Rotation Angle=&RandomRangeFloat(Minimum=0, Maximum=360)
    Sprite Rotation Mode=::NewEnumerator0
    Sprite Size Mode=::NewEnumerator3
    Uniform Sprite Size=37.5
    UsePositionOffset=false
    Write Material Random=true
   )
   &FN_SetSpawnGroup(Number of Spawn Groups=2)
   &SkeletalMeshLocation(
    Bone / Socket Index=7
    Bone Sampling Mode=::NewEnumerator4
    Interpolate Mesh Sampling=false
    MaskSpawnGroup=false
    Orientation Sampling=::NewEnumerator1
    Sampled Position Offset=&Add_Vector(
     A=&Lerp_Vector(
      A=<32 35 0>
      B=<-32 35 0>
      Alpha=&Round_Float(Float:NiagaraFloat=&RandomRangeFloat(Minimum=0, Maximum=1))
     )
     B=<0 0 0>
    )
    Skeletal Mesh=$User.SkeletalMesh
   )
   &AddVelocityInCone(
    Cone Angle=10
    Cone Axis=<0 1 0>
    Module.Use Velocity Falloff On Cone Axis=true
    Velocity Distribution Along Cone Axis=0.75
    Velocity Falloff Away From Cone Axis=0.333
    Velocity Strength=&Multiply_Float(A=&RandomRangeFloat(Minimum=150, Maximum=180), B=1)
   )
   &AddVelocity(Velocity=<0 0 -20>)
   &ShapeLocation(MaskSpawnGroup=false, Override Seed=false, Sphere Radius=2, UseOwnerScale=true) {enabled=false}
   &CameraOffset(Camera Offset Amount=-5)
  ]
  particleUpdateScript [
   &ParticleState()
   &Drag(Drag=&RandomRangeFloat(Minimum=0.8, Maximum=1.2), Use Linear Drag=true, Use Rotational Drag=false)
   &InheritVelocity(Inherited Velocity Amount Scale=<0.5 0.5 0.5>, Speed Limit=true, UseSpeedThreshold=false) {enabled=false}
   &InheritSourceMovement(ApplyForce=false, ApplyPosition=true, ApplyPositionFalloff=false)
   &AccelerationForce(Acceleration=<0 5000 0>, Coordinate Space=::Local)
   &SolveForcesAndVelocity(Clamp Velocity=false, Limit Acceleration=false)
   &SpriteRotationRate(Rotation Rate=&Multiply_Float(
     A=$Particles.InitSpriteRotRate
     B=&FloatFromCurve(
      FloatCurve=DataInterface() {propertyValues {Curve {keys [
          {arriveTangent=-4.0107, interpMode=RCIM_Cubic, leaveTangent=-4.0107, tangentMode=RCTM_User, time=-0.0046, value=1.0113}
          {arriveTangent=-0.2496, interpMode=RCIM_Cubic, leaveTangent=-0.2496, tangentMode=RCTM_User, time=0.5, value=0.16}
         ]}}}
      CurveIndex=$Emitter.Age
      Scale Curve=150
     )
    ))
   &Color(Color=&ColorFromCurve(ColorCurve=DataInterface() {propertyValues {
       AlphaCurve {keys [<at 0 1 RCIM_Cubic>,<at 1 0 RCIM_Cubic>]}
       BlueCurve {keys [<at 0 1 RCIM_Cubic>,<at 1 1 RCIM_Cubic>]}
       GreenCurve {keys [<at 0 1 RCIM_Cubic>,<at 1 1 RCIM_Cubic>]}
       RedCurve {keys [<at 0 1 RCIM_Cubic>,<at 1 1 RCIM_Cubic>]}
      }})) {enabled=false}
   &ScaleColor(
    Color Value To Scale=$Particles.Color
    Scale Alpha=&FloatFromCurve(FloatCurve=DataInterface() {propertyValues {Curve {keys [
         {arriveTangent=-3.5753, interpMode=RCIM_Cubic, leaveTangent=-3.5753, tangentMode=RCTM_User, time=0.25, value=1}
         {interpMode=RCIM_Cubic, tangentMode=RCTM_User, time=1, value=0}
        ]}}})
    ScaleA=true
    ScaleRGB=true
   )
   &ScaleSpriteSize(Scale Factor=&Vector2DFromFloat(Value=&FloatFromCurve(FloatCurve=DataInterface() {propertyValues {Curve {keys [
          {arriveTangent=0.0905, interpMode=RCIM_Cubic, leaveTangent=0.0905, tangentMode=RCTM_User, time=0, value=0.75}
          {arriveTangent=3.1731, interpMode=RCIM_Cubic, leaveTangent=3.1731, tangentMode=RCTM_User, time=0.1, value=1.2573}
          {arriveTangentWeight=0.2498, interpMode=RCIM_Cubic, tangentMode=RCTM_User, time=0.8, value=2}
         ]}}})))
  ]
  rendererClasses [&NiagaraSpriteRendererProperties]
  renderers [{rendererClass=&NiagaraSpriteRendererProperties}]
 }
 Empty002 {
  emitterUpdateScript [
   &EmitterState(
    Inactive Response=::NewEnumerator1
    Life Cycle Mode=::NewEnumerator1
    Loop Behavior=::NewEnumerator1
    Loop Delay=$User.GrowDelay
    Loop Duration Mode=::NewEnumerator1
    UseLoopDelay=true
   )
   &SpawnBurst_Instantaneous(Use Spawn Probability=false, UseLoopCountLimit=false)
   /SetVariables(Emitter.Cam Pos:Vector3f=&ReturnCameraVector(Camera Property=::NewEnumerator0, Coordinate Space=::Simulation)) {isSetParameters=true}
  ]
  particleSpawnScript [
   /SetVariables(Particles.IsFE:NiagaraBool=&SetBoolByFloatComparison(A=&FN_FrontEndFloatModifier(Gameplay=0, Front End=1), B=0.5)) {isSetParameters=true}
   &InitializeParticle(
    Color=&MakeCustomLinearColorFromBool(Bool=$Particles.IsFE, True Color=<rgba 0.5315 0.7584 0.8177 0.15>, False Color=<rgba 0.5315 0.7584 0.8177 0.5>)
    Lifetime=9999999
    Mesh Scale Mode=::NewEnumerator3
    UsePositionOffset=true
    Write Material Random=true
   )
   /SetVariables(Particles.VisibilityTag=&MakeIntFromFloat(Float=&FN_FrontEndFloatModifier(Gameplay=0, Front End=1))) {isSetParameters=true}
  ]
  particleUpdateScript [
   &ParticleState()
   /SetVariables(Particles.InterpolatePos:Vector3f=/Niagara/DynamicInputs/Transforms/TransformPosition(Position=<0 300 0>, Source Space=::Local, Destination Space=::Simulation)) {isSetParameters=true}
   &InterpolateOverTime(Rate Of Change=2, Target Vector Value:Vector3f=$Particles.InterpolatePos, Variable Type=::NewEnumerator2)
   &self:Clamp_02(
    In:Vector3f=&Lerp_Vector(
     A=$Particles.InterpolatePos
     B=$Particles.InterpolateOverTime.InterpolateOverTime_V3_.Moving Average
     Alpha=&ClampFloat(Float=&Divide_Float(A=$Particles.Age, B=3))
    )
    Max Angle=75
   ) {module=Clamp_02}
   /SetVariables(Particles.InterpolateOverTime.InterpolateOverTime_V3_.Moving Average:Vector3f=$Particles.Clamp_02.Clamped) {isSetParameters=true}
   /SetVariables(Particles.Position2:Vector3f=&Add_Vector(
     A=&Add_Vector(A=Linked() {linkedVariable {name=Engine.Owner.Position, type=NiagaraPosition}}, B=<0 0 0>)
     B=&Multiply_VectorByFloat(
      Vector=&NormalizeVector(
       Vector To Normalize=&Subtract_Vector(A=$Particles.Clamp_02.Clamped, B=Linked() {linkedVariable {name=Engine.Owner.Position, type=NiagaraPosition}})
       Safe Normalize=true
       Safe Fallback Threshold=0
       Safe Fallback Vector=<0 0 0.001>
      )
      Float=300
     )
    )) {isSetParameters=true}
   /SetVariables(Particles.MeshOrientation=$Engine.Owner.Rotation) {isSetParameters=true}
   /SetVariables(Particles.OffsetAge:NiagaraFloat=&Subtract_Float(A=$Particles.Age, B=&FN_FrontEndFloatModifier(Gameplay=0.12, Front End=2))) {isSetParameters=true}
   &FN_Timer(Increment Multiplier:NiagaraFloat=&Lerp_Float(A=0.9, B=1.3, Alpha=&ClampFloat(Float=$System.InterpolateOverTime.InterpolateOverTime_Float.Moving Average)))
   &DynamicMaterialParameters(
    Index 0 Param 1=&MakeFloatFromVector(Vector=$Particles.Position2, Channel=::NewEnumerator0)
    Index 0 Param 2=&MakeFloatFromVector(Vector=$Particles.Position2, Channel=::NewEnumerator1)
    Index 0 Param 3=&MakeFloatFromVector(Vector=$Particles.Position2, Channel=::NewEnumerator2)
    Index 0 Param 4=&MakeFloatFromVector(Vector=&TransformVector(Vector=<0 1 0>, Destination Space=::World), Channel=::NewEnumerator0)
    Index 1 Param 1=&MakeFloatFromVector(Vector=&TransformVector(Vector=<0 1 0>, Destination Space=::World), Channel=::NewEnumerator1)
    Index 1 Param 2=&MakeFloatFromVector(Vector=&TransformVector(Vector=<0 1 0>, Destination Space=::World), Channel=::NewEnumerator2)
    Index 1 Param 3=&MakeFloatFromVector(Vector=Linked() {linkedVariable {name=Engine.Owner.Position, type=NiagaraPosition}}, Channel=::NewEnumerator0)
    Index 1 Param 4=&MakeFloatFromVector(Vector=Linked() {linkedVariable {name=Engine.Owner.Position, type=NiagaraPosition}}, Channel=::NewEnumerator1)
    Index 2 Param 1=&MakeFloatFromVector(Vector=Linked() {linkedVariable {name=Engine.Owner.Position, type=NiagaraPosition}}, Channel=::NewEnumerator2)
    Index 2 Param 2=1
    Index 2 Param 4=&FloatFromCurve(
     FloatCurve=DataInterface() {propertyValues {Curve {keys [
         {arriveTangent=0.9379, interpMode=RCIM_Cubic, leaveTangent=0.9379, tangentMode=RCTM_User, time=0, value=0}
         {interpMode=RCIM_Cubic, leaveTangent=1.0524, time=2, value=1}
        ]}}}
     CurveIndex=$Particles.OffsetAge
    )
    Index 3 Param 1=$Particles.FN_Timer.Value
    Index 3 Param 2=1
    Module.Param0WriteEnabled=true
    Module.Param0WriteEnabled1=true
    Module.Param0WriteEnabled2=true
    Module.Param0WriteEnabled3=true
    Module.Param1WriteEnabled=true
    Module.Param1WriteEnabled1=true
    Module.Param1WriteEnabled2=true
    Module.Param1WriteEnabled3=true
    Module.Param2WriteEnabled=true
    Module.Param2WriteEnabled1=true
    Module.Param2WriteEnabled2=true
    Module.Param2WriteEnabled3=true
    Module.Param3WriteEnabled=true
    Module.Param3WriteEnabled1=true
    Module.Param3WriteEnabled2=true
    Module.Param3WriteEnabled3=true
    Write Parameter Index 1=true
    Write Parameter Index 2=true
    Write Parameter Index 3=true
   )
   /SetVariables() {enabled=false, isSetParameters=true}
   &ScaleColor(
    Scale Alpha=&FloatFromCurve(
     FloatCurve=DataInterface() {propertyValues {Curve {keys [
         {arriveTangent=0.3656, interpMode=RCIM_Cubic, leaveTangent=0.3656, tangentMode=RCTM_User, time=1, value=0}
         {interpMode=RCIM_Cubic, tangentMode=RCTM_User, time=2, value=1}
        ]}}}
     CurveIndex=$Particles.OffsetAge
    )
    ScaleA=true
    ScaleRGB=true
   ) {enabled=false}
   &ScaleColor(
    Scale Alpha=&Multiply_Float(
     A=&FloatFromCurve(
      FloatCurve=DataInterface() {propertyValues {Curve {keys [
          {arriveTangent=0.3656, interpMode=RCIM_Cubic, leaveTangent=0.3656, tangentMode=RCTM_User, time=1, value=0}
          {interpMode=RCIM_Cubic, tangentMode=RCTM_User, time=2, value=1}
         ]}}}
      CurveIndex=$Particles.OffsetAge
      Scale Curve=3
     )
     B=&FloatFromCurve(
      FloatCurve=DataInterface() {propertyValues {Curve {keys [<at 0 1 RCIM_Cubic>,<at 1 0.05 RCIM_Cubic>]}}}
      CurveIndex=$System.Timeline.NormalizedTime
     )
    )
    ScaleA=true
    ScaleRGB=true
   ) {module=ScaleColor001}
   /SetVariables(Particles.Position=&Add_Vector(
     A=Linked() {linkedVariable {name=Engine.Owner.Position, type=NiagaraPosition}}
     B=&TransformVector(
      Vector=&MakeVector(
       X=0
       Y=&FloatFromCurve(
        FloatCurve=DataInterface() {propertyValues {Curve {keys [
            {arriveTangent=-0.2987, interpMode=RCIM_Cubic, leaveTangent=-0.2987, tangentMode=RCTM_User, time=0, value=0.75}
            {arriveTangent=0.2159, interpMode=RCIM_Cubic, leaveTangent=0.2159, tangentMode=RCTM_User, time=1, value=0}
           ]}}}
        CurveIndex=&ClampFloat(Float=&Subtract_Float(A=$Particles.Age, B=0.5))
        Scale Curve=-60
       )
       Z=0
      )
      Destination Space=::World
     )
    )) {isSetParameters=true}
   /SetVariables(Particles.Angle:NiagaraFloat=/Game/Effects/Niagara/DynamicInputs/Math/FN_RemapDotProduct(
     SystemAxis:Vector3f=$Engine.Owner.SystemYAxis
     ComparisonAxis:Vector3f=&ReturnCameraVector(Camera Property=::NewEnumerator4, Coordinate Space=::World, Origin Position=$Particles.Position)
     SourceMin:NiagaraFloat=0.95
     SourceMax:NiagaraFloat=1
     DestMin:NiagaraFloat=0
     DestMax:NiagaraFloat=1
    )) {isSetParameters=true}
   /SetVariables(Particles.Energy color:LinearColor=&MultiplyLinearColor(
     Linear Color=<rgba 0.525 1.298 1.5 1>
     Scale Factor=&MakeVector4(
      X=1
      Y=1
      Z=1
      W=&Multiply_Float(
       A=&FloatFromCurve(
        FloatCurve=DataInterface() {propertyValues {Curve {keys [
            {
             arriveTangentWeight=0.7148
             interpMode=RCIM_Cubic
             leaveTangentWeight=0.7148
             tangentMode=RCTM_User
             tangentWeightMode=RCTWM_WeightedBoth
             time=0
             value=1
            }
            {
             arriveTangentWeight=0.7835
             interpMode=RCIM_Cubic
             leaveTangentWeight=0.7835
             tangentMode=RCTM_User
             tangentWeightMode=RCTWM_WeightedBoth
             time=1
             value=0
            }
           ]}}}
        CurveIndex=$System.Timeline.NormalizedTime
       )
       B=&FloatFromCurve(
        FloatCurve=DataInterface() {propertyValues {Curve {keys [{tangentMode=RCTM_User, time=0.3, value=0}, {tangentMode=RCTM_User, time=1.3, value=1}]}}}
        CurveIndex=$Particles.OffsetAge
       )
      )
     )
    )) {isSetParameters=true}
   &ScaleColor(
    Scale Alpha=&Lerp_Float(A=1, B=0, Alpha=&ClampFloat(Float=$Particles.Angle))
    ScaleA=true
    ScaleRGB=true
   ) {module=ScaleColor002}
   &ScaleMeshSize(Scale Factor=&MakeVector(
     X=&FloatFromCurve(FloatCurve=DataInterface() {propertyValues {Curve {keys [<at 2 0>,<at 4 1>]}}}, CurveIndex=$Particles.OffsetAge)
     Y=1
     Z=1
    ))
   &ScaleColor(
    Scale Alpha=&FNTimeOfDayScalar(Night Scale=0.7, Morning Scale=0.3, Evening Scale=0.3)
    ScaleA=true
    ScaleRGB=true
   ) {module=ScaleColor003}
   &ScaleColor(
    Scale RGB=&VectorFromFloat(Value=&MakeCustomFloatFromBool(Bool=$User.IsFrontEnd, True Float=1, False Float=1))
    ScaleA=true
    ScaleRGB=true
   ) {module=ScaleColor004}
   /SetVariables(Particles.Energy Scale FE:Vector3f=&FN_FrontEndVector3Modifier(
     Gameplay=<1 1 1>
     Front End=&VectorFromFloat(Value=&FloatFromCurve(
       FloatCurve=DataInterface() {propertyValues {Curve {keys [{tangentMode=RCTM_User, time=0, value=0}, {tangentMode=RCTM_User, time=1, value=1}]}}}
       CurveIndex=$Particles.OffsetAge
      ))
    )) {isSetParameters=true}
  ]
  rendererClasses [&NiagaraMeshRendererProperties]
  renderers [
   {rendererClass=&NiagaraMeshRendererProperties, rendererIndex=0}
   {rendererClass=&NiagaraMeshRendererProperties, rendererIndex=1}
   {rendererClass=&NiagaraMeshRendererProperties, rendererIndex=2}
   {rendererClass=&NiagaraMeshRendererProperties, rendererIndex=3}
  ]
 }
 Energy_ribbons {
  bEnabled=false
  emitterUpdateScript [
   &EmitterState()
   &SpawnBurst_Instantaneous(Age:NiagaraFloat=$Emitter.Age, Use Spawn Probability=false, UseLoopCountLimit=false)
   &SpawnRate(SpawnRate=4, Use Spawn Probability=false)
   &SpawnRate(SpawnGroup=1, SpawnRate=4, Use Spawn Probability=false) {module=SpawnRate002}
   &SpawnRate(SpawnGroup=2, SpawnRate=5, Use Spawn Probability=false) {enabled=false, module=SpawnRate001}
   &SpawnBurst_Instantaneous(Spawn Time=2.5707, Use Spawn Probability=false, UseLoopCountLimit=false) {enabled=false, module=SpawnBurst_Instantaneous001}
  ]
  particleSpawnScript [
   /SetVariables(Particles.DynMatParm_2:Vector4f=&MakeVector4(X=1, Y=0, Z=0, W=0)) {isSetParameters=true}
   /SetVariables(Particles.PosOffset_2:NiagaraPosition=$Particles.Position) {isSetParameters=true}
   &SystemLocation(MaskSpawnGroup=false) {enabled=false}
   &InitializeRibbon(
    Color=&MultiplyLinearColor(
     Linear Color=$User.Primary Color
     Scale Factor=&MakeVector4_FromFloat(Float=&FloatFromCurve(
       FloatCurve=DataInterface() {propertyValues {Curve {keys [{tangentMode=RCTM_User, time=0.5, value=0}, {tangentMode=RCTM_User, time=5, value=1}]}}}
       CurveIndex=$Emitter.Age
      ))
    )
    Lifetime=1.5
    Ribbon Width=&RandomRangeFloat(Minimum=25, Maximum=100)
    UsePositionOffset=false
    Write Material Random=true
   )
   /SetVariables(Particles.RibbonID=/Niagara/DynamicInputs/NiagaraID/MakeNiagaraIDFromInt(Index:NiagaraInt32=$Emitter.SpawnGroup, Acquire Tag:NiagaraInt32=0)) {isSetParameters=true}
   &SphereLocation(
    MaskSpawnGroup=false
    Override Seed=false
    Sphere Radius=150
    Surface Only=false
    UseNonUniformScale=false
    UseOffset=false
    UseSphereOrientation=false
   )
  ]
  particleUpdateScript [
   &ParticleState()
   &DynamicMaterialParameters(
    Index 3 Param 2=&FloatFromCurve(FloatCurve=DataInterface() {propertyValues {Curve {keys [<at 0 -1>,<at 1 1>]}}})
    Module.Param0WriteEnabled3=true
    Module.Param1WriteEnabled3=true
    Module.Param2WriteEnabled3=true
    Module.Param3WriteEnabled3=true
    Write Parameter Index 0=false
    Write Parameter Index 3=true
   ) {enabled=false}
   &AccelerationForce(
    Acceleration=&Multiply_VectorByFloat(
     Vector=&RandomRangeVector(Minimum=<25 5000 25>, Maximum=<-25 5000 -25>)
     Float=&FloatFromCurve(
      FloatCurve=DataInterface() {propertyValues {Curve {keys [<at -1 2>, {tangentMode=RCTM_User, time=0, value=1}, {tangentMode=RCTM_User, time=1, value=1}]}}}
      CurveIndex=$User.ForwardDot
     )
    )
    Coordinate Space=::Local
   )
   &Drag(Drag=6, Use Linear Drag=true, Use Rotational Drag=false)
   &InheritSourceMovement(
    Applied Position Scale=&VectorFromCurve(VectorCurve=DataInterface() {propertyValues {
       XCurve {keys [
         {interpMode=RCIM_Cubic, leaveTangentWeight=1.0502, tangentMode=RCTM_User, time=0.35, value=1}
         {arriveTangentWeight=1.4157, interpMode=RCIM_Cubic, tangentMode=RCTM_User, time=1, value=0}
        ]}
       YCurve {keys [
         {interpMode=RCIM_Cubic, leaveTangentWeight=1.0502, tangentMode=RCTM_User, time=0.35, value=1}
         {arriveTangentWeight=1.4157, interpMode=RCIM_Cubic, tangentMode=RCTM_User, time=1, value=0}
        ]}
       ZCurve {keys [
         {interpMode=RCIM_Cubic, leaveTangentWeight=1.0502, tangentMode=RCTM_User, time=0.35, value=1}
         {arriveTangentWeight=1.4157, interpMode=RCIM_Cubic, tangentMode=RCTM_User, time=1, value=0}
        ]}
      }})
    ApplyForce=false
    ApplyPosition=true
    ApplyPositionFalloff=false
   )
   &self:DynParam_ColorPicker(
    Primary Color:LinearColor=&FN_Select_Color_from_Time_Of_Day(
     Morning=&HueShiftLinearColor(Hue=0, Saturation=1, Value=0.5, Linear Color=$User.Primary Color)
     Day=$User.Primary Color
     Evening=&HueShiftLinearColor(Hue=0, Saturation=1, Value=0.25, Linear Color=$User.Primary Color)
     Night=&HueShiftLinearColor(Hue=0, Saturation=0.5, Value=1, Linear Color=$User.Primary Color)
    )
    Secondary Color:LinearColor=&FN_Select_Color_from_Time_Of_Day(
     Morning=&HueShiftLinearColor(Hue=0, Saturation=1, Value=0.5, Linear Color=$User.Secondary Color)
     Day=$User.Secondary Color
     Evening=&HueShiftLinearColor(Hue=0, Saturation=1, Value=1, Linear Color=$User.Secondary Color)
     Night=&HueShiftLinearColor(Hue=0.15, Saturation=0.25, Value=1, Linear Color=$User.Secondary Color)
    )
   ) {module=DynParam_ColorPicker}
   /SetVariables(Particles.RibbonLinkOrder:NiagaraFloat=$Particles.Age) {isSetParameters=true}
   &SkeletalMeshLocation(
    Bone / Socket Index=17
    Bone Sampling Mode=::NewEnumerator4
    MaskSpawnGroup=false
    Sampled Position Offset=<100 0 0>
    Skeletal Mesh=$User.SkeletalMesh
    Use Sampled Mesh Attribute as Ribbon ID=false
    Write To Particle Position=false
   ) {module=SkeletalMeshLocation001}
   &SkeletalMeshLocation(
    Bone / Socket Index=16
    Bone Sampling Mode=::NewEnumerator4
    MaskSpawnGroup=false
    Sampled Position Offset Coordinate Space:ENiagaraSkelMeshTransforms=::NewEnumerator0
    Skeletal Mesh=$User.SkeletalMesh
    Use Sampled Mesh Attribute as Ribbon ID=false
    Write To Particle Position=false
   ) {enabled=false, module=SkeletalMeshLocation005}
   /Game/Effects/Niagara/Modules/Curves/FN_HermiteSpline(
    Direction Start:Vector3f=&RandomVector(Vector Scale=250)
    Position Blend:NiagaraFloat=$Particles.NormalizedAge
    Position End:Vector3f=&Add_Vector(
     A=&Add_Vector(
      A=Linked() {linkedVariable {name=Particles.SkeletalMeshLocation005.SampledPosition, type=NiagaraPosition}}
      B=&Multiply_VectorByFloat(
       Vector=&ReturnMeshOrientationAxis(Mesh Orientation Axis=::ZAxis, Mesh Orientation Quat=$Particles.SkeletalMeshLocation005.SampledRotation)
       Float=200
      )
     )
     B=&RandomVector(Vector Scale=100)
    )
    Position Start:NiagaraPosition=$Particles.SkeletalMeshLocation001.SampledPosition
    Write Particles Position=true
   ) {enabled=false}
   /SetVariables(Particles.New Vector 3f:Vector3f=/Niagara/DynamicInputs/Math/LerpMultipleVectors(
     Alpha:NiagaraFloat=$Particles.NormalizedAge
     V 1:Vector3f=Linked() {linkedVariable {name=Particles.SkeletalMeshLocation001.SampledPosition, type=NiagaraPosition}}
     V 2:Vector3f=$Particles.SkeletalMeshLocation002.SampledPosition
     V 3:Vector3f=$Particles.SkeletalMeshLocation003.SampledPosition
     V 4:Vector3f=$Particles.SkeletalMeshLocation004.SampledPosition
     V 5:Vector3f=Linked() {linkedVariable {name=Particles.SkeletalMeshLocation005.SampledPosition, type=NiagaraPosition}}
     Vector Count:ENiagara_MultipleLerpCount=::NewEnumerator2
    )) {enabled=false, isSetParameters=true}
   &DynamicMaterialParameters(
    Index 3 Param 1=1
    Index 3 Param 2=1
    Module.Param0WriteEnabled3=true
    Module.Param1WriteEnabled3=true
    Module.Param2WriteEnabled3=true
    Module.Param3WriteEnabled3=true
    Write Parameter Index 0=false
    Write Parameter Index 3=true
   ) {module=DynamicMaterialParameters001}
   &CurlNoiseForce(Noise Frequency=150, Noise Strength=750, Pan Noise=false)
   &SolveForcesAndVelocity(Clamp Velocity=false, Limit Acceleration=false)
   &ScaleColor(
    Scale Alpha=&Multiply_Float(
     A=&FloatFromCurve(
      FloatCurve=DataInterface() {propertyValues {Curve {keys [
          <at 0.2 0 RCIM_Cubic>
          {arriveTangent=2.3539, interpMode=RCIM_Cubic, leaveTangent=2.3539, time=0.3697, value=1}
          <at 0.6248 1>
          {interpMode=RCIM_Cubic, leaveTangent=-4.7023, time=0.888, value=0}
         ]}}}
      Scale Curve=5
     )
     B=&FloatFromCurve(FloatCurve=DataInterface() {propertyValues {Curve {keys [<at 0 1>,<at 1 0.05>]}}}, CurveIndex=$System.Timeline.NormalizedTime)
    )
    Scale RGB=&VectorFromFloat(Value=1)
    ScaleA=true
    ScaleRGB=true
   )
   &ScaleRibbonWidth(Ribbon Width Scale=&FloatFromCurve(FloatCurve=DataInterface() {propertyValues {Curve {keys [<at 0.1475 0>, {arriveTangent=0.5865, interpMode=RCIM_Cubic, leaveTangent=0.5865, time=0.4107, value=0.95}, <at 1 0.5 RCIM_Cubic>]}}}))
   /Niagara/Modules/Update/Utility/LerpParticleAttributes(
    Lerp Alpha:NiagaraFloat=&FloatFromCurve(FloatCurve=DataInterface() {propertyValues {Curve {keys [<at 0 1>,<at 1 0>]}}})
    Lerp Position=true
    Position A:NiagaraPosition=&ConvertPositionToVector(Input Position=$Engine.Emitter.SimulationPosition, Convert To Absolute Large World Vector=false)
   ) {enabled=false}
   /SetVariables(Particles.New Position:NiagaraPosition=&self:ScratchDynamicInput_03(
     New float:NiagaraFloat=&FloatFromCurve(FloatCurve=DataInterface() {propertyValues {Curve {keys [{tangentMode=RCTM_User, time=0.1, value=0}, {tangentMode=RCTM_User, time=1, value=1}]}}})
     New Position:NiagaraPosition=&ConvertVectorToPosition(
      Input Position=Linked() {linkedVariable {name=Particles.SkeletalMeshLocation001.SampledPosition, type=NiagaraPosition}}
      Vector Is In Absolute Large World Space:NiagaraBool=false
     )
     New Position001:NiagaraPosition=$Particles.Position
    )) {isSetParameters=true}
  ]
  rendererClasses [&NiagaraSpriteRendererProperties, &NiagaraRibbonRendererProperties]
  renderers [
   {rendererClass=&NiagaraSpriteRendererProperties}
   {rendererClass=&NiagaraRibbonRendererProperties, rendererIndex=1}
  ]
 }
 Initial_cloud004 {
  emitterUpdateScript [
   &EmitterState()
   &SpawnRate(SpawnRate=90, Use Spawn Probability=false) {enabled=false}
   &SpawnRate(SpawnRate=&FN_FrontEndFloatModifier(Gameplay=10, Front End=0), Use Spawn Probability=false) {module=SpawnRate003}
   &SpawnRate(
    SpawnGroup=1
    SpawnRate=&FN_FrontEndFloatModifier(Gameplay=10, Front End=0)
    Use Spawn Probability=false
   ) {module=SpawnRate001}
   &SpawnRate(SpawnGroup=2, SpawnRate=2, Use Spawn Probability=false) {enabled=false, module=SpawnRate004}
   &SpawnRate(SpawnGroup=3, SpawnRate=2, Use Spawn Probability=false) {enabled=false, module=SpawnRate005}
   &SpawnRate(SpawnGroup=4, SpawnRate=7.5, Use Spawn Probability=false) {enabled=false, module=SpawnRate002}
   &SpawnBurst_Instantaneous(
    Spawn Group=1
    Spawn Probability=&FN_FrontEndFloatModifier(Gameplay=1, Front End=0)
    Use Spawn Probability=true
    UseLoopCountLimit=false
   )
   &SpawnBurst_Instantaneous(
    Spawn Probability=&FN_FrontEndFloatModifier(Gameplay=1, Front End=0)
    Use Spawn Probability=true
    UseLoopCountLimit=false
   ) {module=SpawnBurst_Instantaneous001}
   &InterpolateOverTime(Target Float Value:NiagaraFloat=$User.RightDot)
  ]
  particleSpawnScript [
   &InitializeParticle(
    Color Maximum=<rgba 0.2953 0.4271 0.5906 0.05>
    Color Minimum=<rgba 0.6571 0.8803 0.9387 0.05>
    Color Mode=::NewEnumerator2
    Lifetime Max=0.75
    Lifetime Min=0.5
    Lifetime Mode=::NewEnumerator1
    Mass Mode=::NewEnumerator3
    Mesh Scale Max=<0.8 0.2 0.2>
    Mesh Scale Min=<0.4 0.1 0.1>
    Mesh Scale Mode=::NewEnumerator2
    Sprite Rotation Mode=::NewEnumerator2
    Sprite Size=&Lerp_Vector2(
     A=<80 150>
     B=<120 150>
     Alpha=&Abs_Float(Float:NiagaraFloat=$Emitter.InterpolateOverTime.InterpolateOverTime_Float.Moving Average)
    )
    Sprite Size Mode=::NewEnumerator4
    Sprite UV Mode=::NewEnumerator3
    UsePositionOffset=false
    Write Material Random=true
   )
   &InitialMeshOrientation(
    Mesh Orientation Mode=::NewEnumerator7
    Rotation=&RandomRangeVector(Minimum=<0 0 0>, Maximum=<1 0 0>)
    Use Rotation Vector=true
   )
   &AddVelocity(Velocity=&RandomRangeVector(Minimum=<0 50 0>, Maximum=<0 0 0>)) {enabled=false}
   /SetVariables(Particles.Raondom Bone:NiagaraInt32=&RandomRangeInt(Minimum=2, Maximum=12)) {isSetParameters=true, module=SetVariables_#1}
   /SetVariables(Particles.CameraOffset2=-2) {isSetParameters=true, module=SetVariables_#2}
   /SetVariables(
    Particles.RibbonID=&MakeNiagaraIDFromInt(Index:NiagaraInt32=$Particles.Raondom Bone, Acquire Tag:NiagaraInt32=0)
    Particles.RibbonWidth=100
   ) {isSetParameters=true, module=SetVariables_#3}
   &SkeletalMeshLocation(
    Bone / Socket Index=&self:ScratchDynamicInput_02(Min Spawn Group:NiagaraInt32=0, Max Spawn Group:NiagaraInt32=65536)
    Bone Sampling Mode=::NewEnumerator4
    MaskSpawnGroup=false
    Orientation Sampling=::NewEnumerator1
    Sampled Position Offset=&Add_Vector(
     A=&Add_Vector(A=&RandomVector(Vector Scale=5), B=<0 0 5>)
     B=&SelectVectorFromArray(
      Direct Array Index=&SpawnGroup()
      Vector Selection Array=DataInterface() {propertyValues {FloatData [<0 0 0>, <0 0 0>, <0 0 0>, <0 0 0>, <0 30 0>]}}
     )
    )
    Skeletal Mesh=$User.SkeletalMesh
    Write To Particle Position:NiagaraBool=&SetBoolByFloatComparison(
     A=&FloatFromCurve(FloatCurve=DataInterface() {propertyValues {Curve {keys [<at 0 1>,<at 0.15 0.5>,<at 0.3 0>]}}})
     B=0.5
    )
   ) {module=SkeletalMeshLocation001}
   /SetVariables(Particles.Velocity=&TransformVector(Vector=<0 1 0>, Destination Space=::Simulation)) {isSetParameters=true, module=SetVariables_#4}
  ]
  particleUpdateScript [
   &ParticleState()
   &ScaleColor(
    Scale Alpha=&Multiply_Float(
     A=&FloatFromCurve(
      FloatCurve=DataInterface() {propertyValues {Curve {keys [
          {interpMode=RCIM_Cubic, leaveTangentWeight=0.0833, tangentMode=RCTM_User, time=0.1803, value=0}
          {arriveTangentWeight=0.1389, interpMode=RCIM_Cubic, leaveTangentWeight=0.2778, tangentMode=RCTM_User, time=0.3443, value=1}
          {
           arriveTangent=-1.7552
           arriveTangentWeight=0.0673
           interpMode=RCIM_Cubic
           leaveTangent=-33.6505
           leaveTangentWeight=1.0091
           tangentMode=RCTM_Break
           tangentWeightMode=RCTWM_WeightedBoth
           time=0.4443
           value=1
          }
          {
           arriveTangent=-0.5272
           arriveTangentWeight=0.3201
           interpMode=RCIM_Cubic
           leaveTangent=-0.5272
           leaveTangentWeight=0.3201
           tangentMode=RCTM_User
           tangentWeightMode=RCTWM_WeightedBoth
           time=1
           value=0
          }
         ]}}}
      Scale Curve=2
     )
     B=&FloatFromCurve(FloatCurve=DataInterface() {propertyValues {Curve {keys [<at 0 1>,<at 1 0.2>]}}}, CurveIndex=$System.Timeline.NormalizedTime)
    )
    Scale RGB=&MakeVectorFromLinearColorRGB(Color=&ColorFromCurve(ColorCurve=DataInterface() {propertyValues {
        AlphaCurve {keys [<at 0 1 RCIM_Cubic>]}
        BlueCurve {keys [<at 0.0435 0.8 RCIM_Cubic>,<at 0.3697 1 RCIM_Cubic>]}
        GreenCurve {keys [<at 0.0435 0.7933 RCIM_Cubic>,<at 0.3697 1 RCIM_Cubic>]}
        RedCurve {keys [<at 0.0435 0.52 RCIM_Cubic>,<at 0.3697 1 RCIM_Cubic>]}
       }}))
    ScaleA=true
    ScaleRGB=true
   )
   /SetVariables(Particles.Smoke Color:LinearColor=&MakeLinearColorFromVectorAndFloat(
     "Vector (RGB)"=<1 1 1>
     "Float (Alpha)"=&FloatFromCurve(FloatCurve=DataInterface() {propertyValues {Curve {keys [
          {arriveTangent=9.0725, interpMode=RCIM_Cubic, leaveTangent=9.0725, tangentMode=RCTM_User, time=0, value=0}
          {interpMode=RCIM_Cubic, leaveTangent=-1.6119, tangentMode=RCTM_Break, time=0.1408, value=0.9847}
          {
           arriveTangent=-1.7491
           arriveTangentWeight=0.8706
           interpMode=RCIM_Cubic
           leaveTangent=-1.7491
           leaveTangentWeight=0.8706
           tangentMode=RCTM_User
           tangentWeightMode=RCTWM_WeightedBoth
           time=0.9
           value=-0.5
          }
         ]}}})
    )) {isSetParameters=true, module=SetVariables_#5}
   &Drag(Drag=6, Use Linear Drag=true, Use Rotational Drag=false)
   &InheritVelocity(
    Inherited Velocity Amount Scale=<-0.025 -0.025 -0.025>
    Speed Limit=false
    UseSpeedThreshold=false
    Velocity Source=&TransformVector(Vector=$Engine.Owner.Velocity, Source Space=::World, Destination Space=::Local)
   ) {enabled=false}
   &InheritSourceMovement(
    Applied Position Scale=&VectorFromFloat(Value=&FloatFromCurve(FloatCurve=DataInterface() {propertyValues {Curve {keys [<at 0 1 RCIM_Cubic>, {arriveTangent=-2.7916, interpMode=RCIM_Cubic, leaveTangent=-2.7916, tangentMode=RCTM_User, time=1, value=0}]}}}))
    ApplyForce=false
    ApplyPosition=true
    ApplyPositionFalloff=false
   )
   &AccelerationForce(
    Acceleration=&Multiply_VectorByFloat(Vector=&RandomRangeVector(Minimum=<25 2500 -980>, Maximum=<-25 5000 -980>), Float=1)
    Coordinate Space=::Local
   )
   &AccelerationForce(Acceleration=<0 7500 0>, Coordinate Space=::Local) {enabled=false, module=AccelerationForce001}
   &PointForce(
    Force Falloff Distance=250
    Force Origin=&Add_Vector(
     A=&ConvertPositionToVector(Input Position=$Engine.Emitter.SimulationPosition, Convert To Absolute Large World Vector=false)
     B=&TransformPosition(Position=<0 60 0>, Source Space=::Simulation, Destination Space=::Local)
    )
    Module.Use Falloff=true
    Module.UseLinearForce=false
   )
   &CurlNoiseForce(
    Curl Noise Cone Mask Axis:Vector3f=&TransformVector(Vector=<0 1 0>, Source Space=::Simulation, Destination Space=::Local)
    Mask Curl Noise=true
    Noise Frequency=15
    Noise Strength=500
    Pan Noise=true
    Pan Noise Field=<1 0 0>
   )
   &SolveForcesAndVelocity(Clamp Velocity=false, Limit Acceleration=false)
   &DynamicMaterialParameters(
    Index 0 Param 1=1
    Index 0 Param 2=1
    Index 0 Param 3=&FloatFromCurve(FloatCurve=DataInterface() {propertyValues {Curve {keys [{arriveTangent=4.42, time=0, value=0}, <at 0.5 1>, <at 0.75 2>]}}})
    Index 0 Param 4=1
    Module.Param0WriteEnabled=true
    Module.Param1WriteEnabled=true
    Module.Param2WriteEnabled=true
    Module.Param3WriteEnabled=true
   )
   /SetVariables(Particles.Scale2:Vector3f=&Multiply_VectorByFloat(Vector=$Particles.Scale, Float=1)) {isSetParameters=true, module=SetVariables_#6}
   &ScaleSpriteSize(
    Non-Uniform Curve Scale=<2 3>
    Non-Uniform Curve Sprite Scale=DataInterface() {propertyValues {
      XCurve {keys [
        {interpMode=RCIM_Cubic, tangentMode=RCTM_User, time=0, value=0.75}
        {arriveTangent=5.9338, arriveTangentWeight=0.3333, interpMode=RCIM_Cubic, leaveTangent=5.9338, tangentMode=RCTM_User, time=1, value=3}
       ]}
      YCurve {keys [
        {
         arriveTangentWeight=0.9222
         interpMode=RCIM_Cubic
         leaveTangentWeight=0.9222
         tangentMode=RCTM_User
         tangentWeightMode=RCTWM_WeightedBoth
         time=0
         value=0.5
        }
        {arriveTangentWeight=0.3333, interpMode=RCIM_Cubic, leaveTangent=1.4376, time=1, value=1}
       ]}
     }}
    Scale Sprite Size Mode=::NewEnumerator3
   )
   &SubUVAnimation(SubUV Loop Count=0.35)
   /SetVariables(Particles.Display Velocity:Vector3f=$Particles.Initial.Velocity) {isSetParameters=true, module=SetVariables_#7}
  ]
  rendererClasses [&NiagaraSpriteRendererProperties]
  renderers [{rendererClass=&NiagaraSpriteRendererProperties}]
 }
 Initial_cloud_FE {
  emitterUpdateScript [
   &EmitterState()
   &SpawnRate(SpawnRate=90, Use Spawn Probability=false) {enabled=false}
   &SpawnRate(SpawnRate=&FN_FrontEndFloatModifier(Gameplay=0, Front End=10), Use Spawn Probability=false) {module=SpawnRate003}
   &SpawnRate(
    SpawnGroup=1
    SpawnRate=&FN_FrontEndFloatModifier(Gameplay=0, Front End=10)
    Use Spawn Probability=false
   ) {module=SpawnRate001}
   &SpawnRate(SpawnGroup=2, SpawnRate=2, Use Spawn Probability=false) {enabled=false, module=SpawnRate004}
   &SpawnRate(SpawnGroup=3, SpawnRate=2, Use Spawn Probability=false) {enabled=false, module=SpawnRate005}
   &SpawnRate(SpawnGroup=4, SpawnRate=7.5, Use Spawn Probability=false) {enabled=false, module=SpawnRate002}
   &SpawnBurst_Instantaneous(
    Spawn Group=1
    Spawn Probability=&FN_FrontEndFloatModifier(Gameplay=0, Front End=1)
    Use Spawn Probability=true
    UseLoopCountLimit=false
   )
   &SpawnBurst_Instantaneous(
    Spawn Probability=&FN_FrontEndFloatModifier(Gameplay=0, Front End=1)
    Use Spawn Probability=true
    UseLoopCountLimit=false
   ) {module=SpawnBurst_Instantaneous001}
   &InterpolateOverTime(Target Float Value:NiagaraFloat=$User.RightDot)
  ]
  particleSpawnScript [
   &InitializeParticle(
    Color Maximum=<rgba 0.2953 0.4271 0.5906 0.05>
    Color Minimum=<rgba 0.6571 0.8803 0.9387 0.05>
    Color Mode=::NewEnumerator2
    Lifetime Max=0.75
    Lifetime Min=0.5
    Lifetime Mode=::NewEnumerator1
    Mass Mode=::NewEnumerator3
    Mesh Scale Max=<0.8 0.2 0.2>
    Mesh Scale Min=<0.4 0.1 0.1>
    Mesh Scale Mode=::NewEnumerator2
    Sprite Rotation Mode=::NewEnumerator2
    Sprite Size=&Lerp_Vector2(
     A=<80 150>
     B=<120 150>
     Alpha=&Abs_Float(Float:NiagaraFloat=$Emitter.InterpolateOverTime.InterpolateOverTime_Float.Moving Average)
    )
    Sprite Size Mode=::NewEnumerator4
    Sprite UV Mode=::NewEnumerator3
    UsePositionOffset=false
    Write Material Random=true
   )
   &InitialMeshOrientation(
    Mesh Orientation Mode=::NewEnumerator7
    Rotation=&RandomRangeVector(Minimum=<0 0 0>, Maximum=<1 0 0>)
    Use Rotation Vector=true
   )
   &AddVelocity(Velocity=&RandomRangeVector(Minimum=<0 50 0>, Maximum=<0 0 0>)) {enabled=false}
   /SetVariables(Particles.Raondom Bone:NiagaraInt32=&RandomRangeInt(Minimum=2, Maximum=12)) {isSetParameters=true, module=SetVariables_#1}
   /SetVariables(Particles.CameraOffset2=-2) {isSetParameters=true, module=SetVariables_#2}
   /SetVariables(
    Particles.RibbonID=&MakeNiagaraIDFromInt(Index:NiagaraInt32=$Particles.Raondom Bone, Acquire Tag:NiagaraInt32=0)
    Particles.RibbonWidth=100
   ) {isSetParameters=true, module=SetVariables_#3}
   &SkeletalMeshLocation(
    Bone / Socket Index=&self:ScratchDynamicInput_02(Min Spawn Group:NiagaraInt32=0, Max Spawn Group:NiagaraInt32=65536)
    Bone Sampling Mode=::NewEnumerator4
    MaskSpawnGroup=false
    Orientation Sampling=::NewEnumerator1
    Sampled Position Offset=&Add_Vector(
     A=&Add_Vector(A=&RandomVector(Vector Scale=5), B=<0 0 5>)
     B=&SelectVectorFromArray(
      Direct Array Index=&SpawnGroup()
      Vector Selection Array=DataInterface() {propertyValues {FloatData [<0 0 0>, <0 0 0>, <0 0 0>, <0 0 0>, <0 30 0>]}}
     )
    )
    Skeletal Mesh=$User.SkeletalMesh
    Write To Particle Position:NiagaraBool=&SetBoolByFloatComparison(
     A=&FloatFromCurve(FloatCurve=DataInterface() {propertyValues {Curve {keys [<at 0 1>,<at 0.15 0.5>,<at 0.3 0>]}}})
     B=0.5
    )
   ) {module=SkeletalMeshLocation001}
   /SetVariables(Particles.Velocity=&TransformVector(Vector=<0 1 0>, Destination Space=::Simulation)) {isSetParameters=true, module=SetVariables_#4}
  ]
  particleUpdateScript [
   &ParticleState()
   &ScaleColor(
    Scale Alpha=&Multiply_Float(
     A=&FloatFromCurve(
      FloatCurve=DataInterface() {propertyValues {Curve {keys [
          {interpMode=RCIM_Cubic, leaveTangentWeight=0.0833, tangentMode=RCTM_User, time=0.1803, value=0}
          {arriveTangentWeight=0.1389, interpMode=RCIM_Cubic, leaveTangentWeight=0.2778, tangentMode=RCTM_User, time=0.3443, value=1}
          {
           arriveTangent=-1.7552
           arriveTangentWeight=0.0673
           interpMode=RCIM_Cubic
           leaveTangent=-33.6505
           leaveTangentWeight=1.0091
           tangentMode=RCTM_Break
           tangentWeightMode=RCTWM_WeightedBoth
           time=0.4443
           value=1
          }
          {
           arriveTangent=-0.5272
           arriveTangentWeight=0.3201
           interpMode=RCIM_Cubic
           leaveTangent=-0.5272
           leaveTangentWeight=0.3201
           tangentMode=RCTM_User
           tangentWeightMode=RCTWM_WeightedBoth
           time=1
           value=0
          }
         ]}}}
      Scale Curve=2
     )
     B=&FloatFromCurve(FloatCurve=DataInterface() {propertyValues {Curve {keys [<at 0 1>,<at 1 0.2>]}}}, CurveIndex=$System.Timeline.NormalizedTime)
    )
    Scale RGB=&MakeVectorFromLinearColorRGB(Color=&ColorFromCurve(ColorCurve=DataInterface() {propertyValues {
        AlphaCurve {keys [<at 0 1 RCIM_Cubic>]}
        BlueCurve {keys [<at 0.0435 0.8 RCIM_Cubic>,<at 0.3697 1 RCIM_Cubic>]}
        GreenCurve {keys [<at 0.0435 0.7933 RCIM_Cubic>,<at 0.3697 1 RCIM_Cubic>]}
        RedCurve {keys [<at 0.0435 0.52 RCIM_Cubic>,<at 0.3697 1 RCIM_Cubic>]}
       }}))
    ScaleA=true
    ScaleRGB=true
   )
   /SetVariables(Particles.Smoke Color:LinearColor=&MakeLinearColorFromVectorAndFloat(
     "Vector (RGB)"=<1 1 1>
     "Float (Alpha)"=&FloatFromCurve(FloatCurve=DataInterface() {propertyValues {Curve {keys [
          {arriveTangent=9.0725, interpMode=RCIM_Cubic, leaveTangent=9.0725, tangentMode=RCTM_User, time=0, value=0}
          {interpMode=RCIM_Cubic, leaveTangent=-1.6119, tangentMode=RCTM_Break, time=0.1408, value=0.9847}
          {
           arriveTangent=-1.7491
           arriveTangentWeight=0.8706
           interpMode=RCIM_Cubic
           leaveTangent=-1.7491
           leaveTangentWeight=0.8706
           tangentMode=RCTM_User
           tangentWeightMode=RCTWM_WeightedBoth
           time=0.9
           value=-0.5
          }
         ]}}})
    )) {isSetParameters=true, module=SetVariables_#5}
   &Drag(Drag=6, Use Linear Drag=true, Use Rotational Drag=false)
   &InheritVelocity(
    Inherited Velocity Amount Scale=<-0.025 -0.025 -0.025>
    Speed Limit=false
    UseSpeedThreshold=false
    Velocity Source=&TransformVector(Vector=$Engine.Owner.Velocity, Source Space=::World, Destination Space=::Local)
   ) {enabled=false}
   &InheritSourceMovement(
    Applied Position Scale=&VectorFromFloat(Value=&FloatFromCurve(FloatCurve=DataInterface() {propertyValues {Curve {keys [<at 0 1 RCIM_Cubic>, {arriveTangent=-2.7916, interpMode=RCIM_Cubic, leaveTangent=-2.7916, tangentMode=RCTM_User, time=1, value=0}]}}}))
    ApplyForce=false
    ApplyPosition=true
    ApplyPositionFalloff=false
   )
   &AccelerationForce(
    Acceleration=&Multiply_VectorByFloat(Vector=&RandomRangeVector(Minimum=<25 2500 -980>, Maximum=<-25 5000 -980>), Float=1)
    Coordinate Space=::Local
   )
   &AccelerationForce(Acceleration=<0 7500 0>, Coordinate Space=::Local) {enabled=false, module=AccelerationForce001}
   &PointForce(
    Force Falloff Distance=250
    Force Origin=&Add_Vector(
     A=&ConvertPositionToVector(Input Position=$Engine.Emitter.SimulationPosition, Convert To Absolute Large World Vector=false)
     B=&TransformPosition(Position=<0 60 0>, Source Space=::Simulation, Destination Space=::Local)
    )
    Module.Use Falloff=true
    Module.UseLinearForce=false
   )
   &CurlNoiseForce(
    Curl Noise Cone Mask Axis:Vector3f=&TransformVector(Vector=<0 1 0>, Source Space=::Simulation, Destination Space=::Local)
    Mask Curl Noise=true
    Noise Frequency=15
    Noise Strength=500
    Pan Noise=true
    Pan Noise Field=<1 0 0>
   )
   &SolveForcesAndVelocity(Clamp Velocity=false, Limit Acceleration=false)
   &DynamicMaterialParameters(
    Index 0 Param 1=1
    Index 0 Param 2=1
    Index 0 Param 3=&FloatFromCurve(FloatCurve=DataInterface() {propertyValues {Curve {keys [{arriveTangent=4.42, time=0, value=0}, <at 0.5 1>, <at 0.75 2>]}}})
    Index 0 Param 4=1
    Module.Param0WriteEnabled=true
    Module.Param1WriteEnabled=true
    Module.Param2WriteEnabled=true
    Module.Param3WriteEnabled=true
   )
   /SetVariables(Particles.Scale2:Vector3f=&Multiply_VectorByFloat(Vector=$Particles.Scale, Float=1)) {isSetParameters=true, module=SetVariables_#6}
   &ScaleSpriteSize(
    Non-Uniform Curve Scale=<2 3>
    Non-Uniform Curve Sprite Scale=DataInterface() {propertyValues {
      XCurve {keys [
        {interpMode=RCIM_Cubic, tangentMode=RCTM_User, time=0, value=0.75}
        {arriveTangent=5.9338, arriveTangentWeight=0.3333, interpMode=RCIM_Cubic, leaveTangent=5.9338, tangentMode=RCTM_User, time=1, value=3}
       ]}
      YCurve {keys [
        {
         arriveTangentWeight=0.9222
         interpMode=RCIM_Cubic
         leaveTangentWeight=0.9222
         tangentMode=RCTM_User
         tangentWeightMode=RCTWM_WeightedBoth
         time=0
         value=0.5
        }
        {arriveTangentWeight=0.3333, interpMode=RCIM_Cubic, leaveTangent=1.4376, time=1, value=1}
       ]}
     }}
    Scale Sprite Size Mode=::NewEnumerator3
   )
   &SubUVAnimation(SubUV Loop Count=0.35)
   /SetVariables(Particles.Display Velocity:Vector3f=$Particles.Initial.Velocity) {isSetParameters=true, module=SetVariables_#7}
  ]
  rendererClasses [&NiagaraSpriteRendererProperties]
  renderers [{rendererClass=&NiagaraSpriteRendererProperties}]
 }
 Ribbons001 {
  bEnabled=false
  emitterUpdateScript [
   &EmitterState()
   &SpawnRate(SpawnRate=90, Use Spawn Probability=false) {enabled=false}
   &SpawnRate(SpawnRate=2.5, Use Spawn Probability=false) {module=SpawnRate003}
   &SpawnRate(SpawnGroup=1, SpawnRate=2.5, Use Spawn Probability=false) {module=SpawnRate001}
   &SpawnRate(
    SpawnGroup=2
    SpawnRate=&MakeCustomFloatFromBool(
     Bool=&IsPlatformSetActive(New Platform Set=DataInterface() {propertyValues {Platforms {qualityLevelMask=-29}}})
     True Float=0
     False Float=2.5
    )
    Use Spawn Probability=false
   ) {module=SpawnRate002}
   &SpawnBurst_Instantaneous(Spawn Group=1, Use Spawn Probability=false, UseLoopCountLimit=false)
   &SpawnBurst_Instantaneous(Use Spawn Probability=false, UseLoopCountLimit=false) {module=SpawnBurst_Instantaneous001}
   /SetVariables(Emitter.Cam Pos:Vector3f=&ReturnCameraVector(Camera Property=::NewEnumerator0, Coordinate Space=::Simulation)) {isSetParameters=true}
  ]
  particleSpawnScript [
   &InitializeParticle(
    Color Maximum=<rgba 0.2 0.3338 0.5 1>
    Color Minimum=<rgba 0.65 0.9275 1 1>
    Color Mode=::NewEnumerator2
    Lifetime Max=3
    Lifetime Min=3
    Lifetime Mode=::NewEnumerator1
    Mesh Scale Max=<1 1.25 1.25>
    Mesh Scale Min=<0.75 1 1>
    Mesh Scale Mode=::NewEnumerator2
    Sprite Size Mode=::NewEnumerator1
    Uniform Sprite Size Max=12
    Uniform Sprite Size Min=6
    UsePositionOffset=true
    Write Material Random=true
   )
   &InitialMeshOrientation(
    Mesh Orientation Mode=::NewEnumerator7
    Rotation=&RandomRangeVector(Minimum=<0 0 0>, Maximum=<1 0 0>)
    Use Rotation Vector=true
   ) {enabled=false}
   &AddVelocity(Velocity=&RandomRangeVector(Minimum=<0 50 0>, Maximum=<0 0 0>)) {enabled=false}
   /SetVariables(Particles.Raondom Bone:NiagaraInt32=&RandomRangeInt(Minimum=6, Maximum=6)) {isSetParameters=true, module=SetVariables_#1}
   /SetVariables(Particles.CameraOffset2=-2) {isSetParameters=true, module=SetVariables_#2}
   /SetVariables(
    Particles.RibbonID=&MakeNiagaraIDFromInt(Index:NiagaraInt32=$Particles.Raondom Bone, Acquire Tag:NiagaraInt32=0)
    Particles.RibbonWidth=75
   ) {isSetParameters=true, module=SetVariables_#3}
   /SetVariables(
    Particles.RibbonU0Override=&Add_Float(
     A=&Multiply_Float(A=$Emitter.Age, B=0.45)
     B=&Divide_Float(A=&MakeFloatFromInt(Int=&SpawnGroup()), B=3)
    )
    Particles.RibbonU1Override:NiagaraFloat=&Add_Float(
     A=&Multiply_Float(A=$Emitter.Age, B=0.45)
     B=&Divide_Float(A=&MakeFloatFromInt(Int=&SpawnGroup()), B=3)
    )
   ) {isSetParameters=true, module=SetVariables_#8}
   /SetVariables() {isSetParameters=true}
  ]
  particleUpdateScript [
   &ParticleState()
   /SetVariables(Particles.Smoke Color:LinearColor=&MakeLinearColorFromVectorAndFloat(
     "Vector (RGB)"=<1 1 1>
     "Float (Alpha)"=&FloatFromCurve(FloatCurve=DataInterface() {propertyValues {Curve {keys [
          {arriveTangent=9.0725, interpMode=RCIM_Cubic, leaveTangent=9.0725, tangentMode=RCTM_User, time=0, value=0}
          {interpMode=RCIM_Cubic, leaveTangent=-1.6119, tangentMode=RCTM_Break, time=0.1408, value=0.9847}
          {
           arriveTangentWeight=0.4305
           interpMode=RCIM_Cubic
           leaveTangentWeight=0.4305
           tangentMode=RCTM_User
           tangentWeightMode=RCTWM_WeightedBoth
           time=0.9
           value=-0.5
          }
         ]}}})
    )) {isSetParameters=true, module=SetVariables_#5}
   &Drag(Drag=6, Use Linear Drag=true, Use Rotational Drag=false)
   &AccelerationForce(Acceleration=<0 5000 0>, Coordinate Space=::Local)
   &PointForce(
    Force Falloff Distance=250
    Force Origin=&Add_Vector(
     A=&ConvertPositionToVector(Input Position=$Engine.Emitter.SimulationPosition, Convert To Absolute Large World Vector=false)
     B=&TransformPosition(Position=<0 60 0>, Source Space=::Simulation, Destination Space=::Local)
    )
    Force Strength=200
    Module.Use Falloff=true
    Module.UseLinearForce=false
   )
   &CurlNoiseForce(
    Curl Noise Cone Mask Axis:Vector3f=&TransformVector(Vector=<0 1 0>, Source Space=::Simulation, Destination Space=::Local)
    Mask Curl Noise=true
    Noise Frequency=15
    Noise Strength=500
    Pan Noise=true
    Pan Noise Field=<1 0 0>
   )
   &InheritVelocity(
    Inherited Velocity Amount Scale=&VectorFromFloat(Value=&FloatFromCurve(FloatCurve=DataInterface() {propertyValues {Curve {keys [<at 0.14 0>,<at 0.15 0.25>,<at 0.2 0>]}}}))
    Speed Limit=true
    UseSpeedThreshold=false
   )
   &InheritSourceMovement(
    Applied Position Scale=&VectorFromCurve(VectorCurve=DataInterface() {propertyValues {XCurve {keys [<at 0 1>,<at 1 0>]}, YCurve {keys [<at 0 1>,<at 1 0>]}, ZCurve {keys [<at 0 1>,<at 1 0>]}}})
    ApplyForce=false
    ApplyPosition=true
    ApplyPositionFalloff=false
   )
   &SolveForcesAndVelocity(Clamp Velocity=false, Limit Acceleration=false)
   &SkeletalMeshLocation(
    Bone / Socket Index=&MaskIntBySpawnGroup(
     Masked Int=0
     Passthrough Int=&MaskIntBySpawnGroup(Masked Int=1, Passthrough Int=$Particles.Initial.Raondom Bone, Spawn Group Mask=1)
     Spawn Group Mask=0
    )
    Bone Sampling Mode=::NewEnumerator4
    MaskSpawnGroup=false
    Orientation Sampling=::NewEnumerator1
    Sampled Position Offset=&Add_Vector(A=&RandomVector(Vector Scale=5), B=<100 0 5>)
    Skeletal Mesh=$User.SkeletalMesh
    Write To Particle Position:NiagaraBool=&SetBoolByFloatComparison(
     A=&FloatFromCurve(FloatCurve=DataInterface() {propertyValues {Curve {keys [<at 0 1>,<at 0.1 0.5>,<at 0.3 0>]}}})
     B=0.5
    )
   ) {module=SkeletalMeshLocation001}
   /SetVariables(Particles.Ribbon dynamic:Vector4f=&MakeVector4(X=0, Y=0, Z=0, W=0)) {isSetParameters=true}
   &ScaleRibbonWidth(Ribbon Width Scale=&FloatFromCurve(
     FloatCurve=DataInterface() {propertyValues {Curve {keys [
         {arriveTangent=0.6631, interpMode=RCIM_Cubic, leaveTangent=0.6631, tangentMode=RCTM_User, time=0, value=0.25}
         {interpMode=RCIM_Cubic, tangentMode=RCTM_User, time=0.5, value=1}
         {arriveTangent=-0.6203, interpMode=RCIM_Cubic, leaveTangent=-0.6203, tangentMode=RCTM_User, time=1, value=0}
        ]}}}
     Scale Curve=3
    ))
   &DynamicMaterialParameters(Module.Param0WriteEnabled=true, Module.Param1WriteEnabled=true, Module.Param2WriteEnabled=true, Module.Param3WriteEnabled=true)
   &ScaleColor(
    Scale Alpha=&Multiply_Float(
     A=&Multiply_Float(
      A=&Multiply_Float(
       A=&Multiply_Float(
        A=&FloatFromCurve(FloatCurve=DataInterface() {propertyValues {Curve {keys [
             {arriveTangent=9.0725, interpMode=RCIM_Cubic, leaveTangent=9.0725, tangentMode=RCTM_User, time=0.1186, value=0}
             {arriveTangent=-3.8189, interpMode=RCIM_Cubic, leaveTangent=-3.8189, tangentMode=RCTM_User, time=0.3186, value=1}
             {
              arriveTangentWeight=0.2864
              interpMode=RCIM_Cubic
              leaveTangent=-0.1866
              tangentMode=RCTM_User
              tangentWeightMode=RCTWM_WeightedBoth
              time=1
              value=-0.1
             }
            ]}}})
        B=1
       )
       B=1
      )
      B=&FloatFromCurve(FloatCurve=DataInterface() {propertyValues {Curve {keys [<at 0 1>,<at 1 0.1>]}}}, CurveIndex=$System.Timeline.NormalizedTime)
     )
     B=1
    )
    ScaleA=true
    ScaleRGB=true
   )
   /SetVariables(
    Particles.RibbonU0Override=&Add_Float(A=$Particles.Initial.RibbonU0Override, B=&Multiply_Float(A=$Emitter.Age, B=0.5))
    Particles.RibbonU1Override:NiagaraFloat=&Add_Float(A=$Particles.Initial.RibbonU1Override, B=&Multiply_Float(A=$Emitter.Age, B=0.5))
   ) {enabled=false, isSetParameters=true}
   /SetVariables(Particles.RibbonLinkOrder:NiagaraFloat=$Particles.Age) {isSetParameters=true}
  ]
  rendererClasses [&NiagaraRibbonRendererProperties, &NiagaraSpriteRendererProperties]
  renderers [
   {rendererClass=&NiagaraRibbonRendererProperties}
   {rendererClass=&NiagaraSpriteRendererProperties, rendererIndex=1}
  ]
 }
 SmokeBody {
  emitterUpdateScript [
   &EmitterState()
   &SpawnRate(
    SpawnRate=&Multiply_Float(
     A=30
     B=&FloatFromCurve(FloatCurve=DataInterface() {propertyValues {Curve {keys [<at 0 1>,<at 1 0.5>]}}}, CurveIndex=$System.Timeline.NormalizedTime)
    )
    Use Spawn Probability=false
   )
  ]
  particleSpawnScript [
   &InitializeParticle(
    Color Maximum=<rgba 0.589 0.8998 1 1>
    Color Minimum=<rgba 0.3417 0.815 0.9387 1>
    Color Mode=::NewEnumerator2
    Lifetime=&Lerp_Float(
     A=0.5
     B=0.8
     Alpha=&ClampFloat(Float=&Divide_Float(A=&VectorLength(Vector=$Engine.Owner.Velocity), B=2200))
    )
    Sprite Rotation Mode=::NewEnumerator2
    Sprite Size Mode=::NewEnumerator3
    Uniform Sprite Size=60
    UsePositionOffset=false
    Write Material Random=true
   )
   &FN_SetSpawnGroup(Number of Spawn Groups=2) {enabled=false}
   &SkeletalMeshLocation(
    Bone / Socket Index=7
    Bone Sampling Mode=::NewEnumerator4
    Interpolate Mesh Sampling=false
    MaskSpawnGroup=false
    Orientation Sampling=::NewEnumerator1
    Sampled Position Offset=<0 -5 0>
    Skeletal Mesh=$User.SkeletalMesh
   )
   &AddVelocityInCone(
    Cone Angle=10
    Cone Axis=<0 1 0>
    Module.Use Velocity Falloff On Cone Axis=true
    Velocity Distribution Along Cone Axis=0.75
    Velocity Falloff Away From Cone Axis=0.333
    Velocity Strength=&Lerp_Float(
     A=500
     B=1000
     Alpha=&FloatFromCurve(FloatCurve=DataInterface() {propertyValues {Curve {keys [<at 0 1>,<at 1 0>]}}}, CurveIndex=$System.Timeline.NormalizedTime)
    )
   )
   &AddVelocity(Velocity=<0 0 5>) {enabled=false}
   &ShapeLocation(MaskSpawnGroup=false, Override Seed=false, Sphere Radius=10, UseOwnerScale=true)
   &AddVelocity(Rotation Coordinate Space=::World, Velocity=<0 0 -50>) {module=AddVelocity001}
  ]
  particleUpdateScript [
   &ParticleState()
   &Drag(Drag=&RandomRangeFloat(Minimum=1, Maximum=0.8), Use Linear Drag=true, Use Rotational Drag=false)
   &InheritSourceMovement(
    Applied Position Scale=&VectorFromFloat(Value=&FloatFromCurve(FloatCurve=DataInterface() {propertyValues {Curve {keys [<at 0.75 1 RCIM_Cubic>, {arriveTangent=-0.4692, interpMode=RCIM_Cubic, leaveTangent=-0.4692, tangentMode=RCTM_User, time=1, value=0.5}]}}}))
    ApplyForce=false
    ApplyPosition=true
    ApplyPositionFalloff=false
   ) {enabled=false}
   &InheritVelocity(
    Inherited Velocity Amount Scale=&MakeVector(
     X=&FloatFromCurve(FloatCurve=DataInterface() {propertyValues {Curve {keys [{tangentMode=RCTM_User, time=0, value=0}, {tangentMode=RCTM_User, time=1, value=1}]}}})
     Y=&FloatFromCurve(FloatCurve=DataInterface() {propertyValues {Curve {keys [{tangentMode=RCTM_User, time=0, value=0}, {tangentMode=RCTM_User, time=1, value=1}]}}})
     Z=0
    )
    Speed Limit=true
    UseSpeedThreshold=false
    Velocity Source=&TransformVector(
     Vector=&Multiply_VectorByFloat(Vector=$Engine.Owner.Velocity, Float=-5)
     Source Space=::World
     Destination Space=::Simulation
    )
   )
   &GravityForce(Gravity=<0 0 -100>)
   &SolveForcesAndVelocity(Clamp Velocity=false, Limit Acceleration=false)
   &SpriteRotationRate(Rotation Rate=&Multiply_Float(
     A=$Particles.InitSpriteRotRate
     B=&FloatFromCurve(
      FloatCurve=DataInterface() {propertyValues {Curve {keys [
          {arriveTangent=-4.0107, interpMode=RCIM_Cubic, leaveTangent=-4.0107, tangentMode=RCTM_User, time=-0.0046, value=1.0113}
          {arriveTangent=-0.2496, interpMode=RCIM_Cubic, leaveTangent=-0.2496, tangentMode=RCTM_User, time=0.5, value=0.16}
         ]}}}
      CurveIndex=$Emitter.Age
      Scale Curve=150
     )
    ))
   &ScaleSpriteSize(Scale Factor=&Vector2DFromFloat(Value=&FloatFromCurve(
      FloatCurve=DataInterface() {propertyValues {Curve {keys [
          {arriveTangent=1.2865, interpMode=RCIM_Cubic, leaveTangent=1.2865, tangentMode=RCTM_User, time=0, value=0.5}
          {arriveTangent=1.875, interpMode=RCIM_Cubic, leaveTangent=1.875, time=0.5694, value=1}
          {arriveTangent=1.7547, interpMode=RCIM_Cubic, leaveTangent=1.7547, tangentMode=RCTM_User, time=0.8, value=1.5}
         ]}}}
      Scale Curve=2.2
     )))
   &DynamicMaterialParameters(
    Index 0 Param 1=&FloatFromCurve(FloatCurve=DataInterface() {propertyValues {Curve {keys [{tangentMode=RCTM_User, time=0, value=0}, {tangentMode=RCTM_User, time=1, value=0.4}]}}})
    Index 0 Param 3=1
    Index 0 Param 4=&FloatFromCurve(FloatCurve=DataInterface() {propertyValues {Curve {keys [<at 0 1>,<at 1 0>]}}})
    Index 2 Param 1=&FloatFromCurve(FloatCurve=DataInterface() {propertyValues {Curve {keys [
         {arriveTangent=2.646, interpMode=RCIM_Cubic, leaveTangent=2.646, tangentMode=RCTM_User, time=0, value=0}
         {leaveTangent=-0.0719, time=1, value=1}
        ]}}})
    Index 2 Param 2=&FloatFromCurve(
     FloatCurve=DataInterface() {propertyValues {Curve {keys [
         {arriveTangent=1.9056, time=0, value=0.2}
         {arriveTangent=0.0259, interpMode=RCIM_Cubic, leaveTangent=0.0259, tangentMode=RCTM_User, time=1, value=1.4}
        ]}}}
     Scale Curve=&RandomRangeFloat(Minimum=0.5, Maximum=1)
    )
    Index 2 Param 3=&FloatFromCurve(FloatCurve=DataInterface() {propertyValues {Curve {keys [
         {interpMode=RCIM_Cubic, leaveTangent=2.3064, tangentMode=RCTM_Break, time=0, value=1}
         {interpMode=RCIM_Cubic, leaveTangent=11.1233, tangentMode=RCTM_Break, time=1, value=2}
        ]}}})
    Index 2 Param 4=&FloatFromCurve(
     FloatCurve=DataInterface() {propertyValues {Curve {keys [
         {arriveTangent=-0.6867, interpMode=RCIM_Cubic, leaveTangent=-0.6867, tangentMode=RCTM_User, time=0, value=1}
         {interpMode=RCIM_Cubic, leaveTangent=-1.2956, tangentMode=RCTM_Break, time=1, value=0.75}
        ]}}}
     Scale Curve=1.6
    )
    Module.Param0WriteEnabled=true
    Module.Param0WriteEnabled1=true
    Module.Param0WriteEnabled2=true
    Module.Param1WriteEnabled=false
    Module.Param1WriteEnabled1=true
    Module.Param1WriteEnabled2=true
    Module.Param2WriteEnabled=true
    Module.Param2WriteEnabled1=true
    Module.Param2WriteEnabled2=true
    Module.Param3WriteEnabled=true
    Module.Param3WriteEnabled1=true
    Module.Param3WriteEnabled2=true
    Write Parameter Index 1=true
    Write Parameter Index 2=true
   )
   &ScaleColor(
    Scale Alpha=&FloatFromCurve(
     FloatCurve=DataInterface() {propertyValues {Curve {keys [<at 0 1 RCIM_Cubic>, {arriveTangent=-3.2209, interpMode=RCIM_Cubic, leaveTangent=-3.2209, tangentMode=RCTM_User, time=0.75, value=0}]}}}
     Scale Curve=&FN_FrontEndFloatModifier(Gameplay=1, Front End=0.1)
    )
    ScaleA=true
    ScaleRGB=true
   )
  ]
  rendererClasses [&NiagaraSpriteRendererProperties]
  renderers [
   {rendererClass=&NiagaraSpriteRendererProperties}
   {rendererClass=&NiagaraSpriteRendererProperties, rendererIndex=1}
  ]
 }
 SmokeSides {
  emitterUpdateScript [
   &EmitterState()
   &SpawnRate(SpawnRate=10, Use Spawn Probability=false)
   &SpawnRate(SpawnGroup=1, SpawnRate=10, Use Spawn Probability=false) {module=SpawnRate001}
  ]
  particleSpawnScript [
   &InitializeParticle(
    Color Maximum=<rgba 0.5744 0.8472 1 1>
    Color Minimum=<rgba 0.2646 0.6695 0.9387 1>
    Color Mode=::NewEnumerator2
    Lifetime Max=0.7
    Lifetime Min=0.6
    Lifetime Mode=::NewEnumerator1
    Sprite Rotation Mode=::NewEnumerator2
    Sprite Size Mode=::NewEnumerator3
    Uniform Sprite Size=45
    UsePositionOffset=false
    Write Material Random=true
   )
   &FN_SetSpawnGroup(Number of Spawn Groups=2) {enabled=false}
   &SkeletalMeshLocation(
    Bone / Socket Index=7
    Bone Sampling Mode=::NewEnumerator4
    Interpolate Mesh Sampling=false
    MaskSpawnGroup=false
    Orientation Sampling=::NewEnumerator1
    Sampled Position Offset=&Add_Vector(
     A=&Lerp_Vector(
      A=<32 25 0>
      B=<-32 25 0>
      Alpha=&Round_Float(Float:NiagaraFloat=&RandomRangeFloat(Minimum=0, Maximum=1))
     )
     B=<0 0 0>
    )
    Skeletal Mesh=$User.SkeletalMesh
   )
   &AddVelocityInCone(
    Cone Angle=10
    Cone Axis=<0 1 0>
    Module.Use Velocity Falloff On Cone Axis=true
    Velocity Distribution Along Cone Axis=0.75
    Velocity Falloff Away From Cone Axis=0.333
    Velocity Strength=&Multiply_Float(A=&RandomRangeFloat(Minimum=150, Maximum=180), B=3)
   )
   &AddVelocity(Velocity=<0 0 5>)
   &ShapeLocation(MaskSpawnGroup=false, Override Seed=false, Sphere Radius=5, UseOwnerScale=true)
  ]
  particleUpdateScript [
   &ParticleState()
   &Drag(Drag=&RandomRangeFloat(Minimum=0.8, Maximum=1.2), Use Linear Drag=true, Use Rotational Drag=false)
   &InheritSourceMovement(ApplyForce=false, ApplyPosition=true, ApplyPositionFalloff=false) {enabled=false}
   &SolveForcesAndVelocity(Clamp Velocity=false, Limit Acceleration=false)
   &SpriteRotationRate(Rotation Rate=&Multiply_Float(
     A=$Particles.InitSpriteRotRate
     B=&FloatFromCurve(
      FloatCurve=DataInterface() {propertyValues {Curve {keys [
          {arriveTangent=-4.0107, interpMode=RCIM_Cubic, leaveTangent=-4.0107, tangentMode=RCTM_User, time=-0.0046, value=1.0113}
          {arriveTangent=-0.2496, interpMode=RCIM_Cubic, leaveTangent=-0.2496, tangentMode=RCTM_User, time=0.5, value=0.16}
         ]}}}
      CurveIndex=$Emitter.Age
      Scale Curve=150
     )
    ))
   &ScaleColor(
    Color Value To Scale=$Particles.Color
    Scale Alpha=&FloatFromCurve(
     FloatCurve=DataInterface() {propertyValues {Curve {keys [
         {arriveTangent=9.0725, interpMode=RCIM_Cubic, leaveTangent=9.0725, tangentMode=RCTM_User, time=0, value=0}
         {arriveTangent=-2.5249, interpMode=RCIM_Cubic, leaveTangent=-2.5249, tangentMode=RCTM_User, time=0.1408, value=0.9847}
         {arriveTangent=-0.0405, interpMode=RCIM_Cubic, leaveTangent=-0.0405, tangentMode=RCTM_User, time=1, value=0}
        ]}}}
     Scale Curve=&FN_FrontEndFloatModifier(Gameplay=1.5, Front End=1.25)
    )
    ScaleA=true
    ScaleRGB=true
   )
   &ScaleSpriteSize(Scale Factor=&Vector2DFromFloat(Value=&FloatFromCurve(
      FloatCurve=DataInterface() {propertyValues {Curve {keys [
          {arriveTangent=0.0905, interpMode=RCIM_Cubic, leaveTangent=0.0905, tangentMode=RCTM_User, time=0, value=0.5}
          {arriveTangent=1.875, interpMode=RCIM_Cubic, leaveTangent=1.875, time=0.1308, value=1.4523}
          {arriveTangent=0.3828, interpMode=RCIM_Cubic, leaveTangent=0.3828, tangentMode=RCTM_User, time=0.8, value=2}
         ]}}}
      Scale Curve=0.75
     )))
   &DynamicMaterialParameters(
    Index 0 Param 3=&FloatFromCurve(
     FloatCurve=DataInterface() {propertyValues {Curve {keys [{arriveTangent=-53.9199, time=0, value=1}, {leaveTangent=0.2723, time=0.25, value=1}]}}}
     Scale Curve=0.1
    )
    Index 0 Param 4=&FloatFromCurve(FloatCurve=DataInterface() {propertyValues {Curve {keys [<at 0 1>,<at 1 0>]}}})
    Index 1 Param 1=&FloatFromCurve(FloatCurve=DataInterface() {propertyValues {Curve {keys [<at 0 0>, {arriveTangent=4.2257, interpMode=RCIM_Cubic, leaveTangent=4.2257, tangentMode=RCTM_User, time=0.25, value=0}]}}})
    Index 1 Param 2=&FloatFromCurve(FloatCurve=DataInterface() {propertyValues {Curve {keys [<at 0 0>]}}})
    Index 1 Param 3=&FloatFromCurve(FloatCurve=DataInterface() {propertyValues {Curve {keys [{interpMode=RCIM_Cubic, leaveTangent=6.3917, tangentMode=RCTM_Break, time=0, value=0}]}}})
    Index 1 Param 4=&FloatFromCurve(FloatCurve=DataInterface() {propertyValues {Curve {keys [{arriveTangent=-0.136, leaveTangent=-0.136, time=0, value=1}]}}})
    Index 2 Param 1=&FloatFromCurve(FloatCurve=DataInterface() {propertyValues {Curve {keys [
         {arriveTangent=2.646, interpMode=RCIM_Cubic, leaveTangent=2.646, tangentMode=RCTM_User, time=0, value=0}
         {leaveTangent=-0.0719, time=1, value=1}
        ]}}})
    Index 2 Param 2=&FloatFromCurve(
     FloatCurve=DataInterface() {propertyValues {Curve {keys [
         {arriveTangent=1.9056, time=0, value=0.2}
         {arriveTangent=0.0259, interpMode=RCIM_Cubic, leaveTangent=0.0259, tangentMode=RCTM_User, time=1, value=1.4}
        ]}}}
     Scale Curve=&RandomRangeFloat(Minimum=0.5, Maximum=1)
    )
    Index 2 Param 3=&FloatFromCurve(FloatCurve=DataInterface() {propertyValues {Curve {keys [
         {interpMode=RCIM_Cubic, leaveTangent=2.3064, tangentMode=RCTM_Break, time=0, value=1}
         {interpMode=RCIM_Cubic, leaveTangent=11.1233, tangentMode=RCTM_Break, time=1, value=2}
        ]}}})
    Index 2 Param 4=&FloatFromCurve(
     FloatCurve=DataInterface() {propertyValues {Curve {keys [
         {arriveTangent=-0.6867, interpMode=RCIM_Cubic, leaveTangent=-0.6867, tangentMode=RCTM_User, time=0, value=1}
         {interpMode=RCIM_Cubic, leaveTangent=-1.2956, tangentMode=RCTM_Break, time=1, value=0.75}
        ]}}}
     Scale Curve=1.6
    )
    Module.Param0WriteEnabled=true
    Module.Param0WriteEnabled1=true
    Module.Param0WriteEnabled2=true
    Module.Param1WriteEnabled=false
    Module.Param1WriteEnabled1=true
    Module.Param1WriteEnabled2=true
    Module.Param2WriteEnabled=true
    Module.Param2WriteEnabled1=true
    Module.Param2WriteEnabled2=true
    Module.Param3WriteEnabled=true
    Module.Param3WriteEnabled1=true
    Module.Param3WriteEnabled2=true
    Write Parameter Index 1=true
    Write Parameter Index 2=true
   )
  ]
  rendererClasses [&NiagaraSpriteRendererProperties]
  renderers [
   {rendererClass=&NiagaraSpriteRendererProperties}
   {rendererClass=&NiagaraSpriteRendererProperties, rendererIndex=1}
  ]
 }
}
summary {
 systemName=NS_snowfall_Glider
 emitters [
  {
   bEnabled=false
   emitterName=Ribbons001
   rendererClasses [&NiagaraRibbonRendererProperties, &NiagaraSpriteRendererProperties]
  }
  {emitterName=Initial_cloud004, rendererClasses [&NiagaraSpriteRendererProperties]}
  {emitterName=SmokeSides, rendererClasses [&NiagaraSpriteRendererProperties]}
  {bEnabled=false, emitterName=Clumps001, rendererClasses [&NiagaraSpriteRendererProperties]}
  {bEnabled=false, emitterName=Chunks001, rendererClasses [&NiagaraMeshRendererProperties]}
  {
   bEnabled=false
   emitterName=Energy_ribbons
   rendererClasses [&NiagaraSpriteRendererProperties, &NiagaraRibbonRendererProperties]
  }
  {emitterName=SmokeBody, rendererClasses [&NiagaraSpriteRendererProperties]}
  {emitterName=Empty002, rendererClasses [&NiagaraMeshRendererProperties]}
  {emitterName=Initial_cloud_FE, rendererClasses [&NiagaraSpriteRendererProperties]}
 ]
 userVariables [
  User.bIsFrontEnd {type=NiagaraBool}
  User.bIsFrontEndPreview {type=NiagaraBool}
  User.ForwardDot {type=NiagaraFloat, defaultValue=0, description=""}
  User.GrowDelay {type=NiagaraFloat, defaultValue=0.4, description=""}
  User.IsFrontEnd {type=NiagaraBool}
  User.IsWhiteVersion {type=NiagaraBool, defaultValue=true}
  User.Primary Color {type=LinearColor, description="", defaultValue {struct=LinearColor, value=<rgba 0.3755 0.822 0.9387 1>}}
  User.RightDot {type=NiagaraFloat, defaultValue=0, description=""}
  User.Secondary Color {type=LinearColor, description="", defaultValue {struct=LinearColor, value=<rgba 0.3755 0.822 0.9387 1>}}
  User.SkeletalMesh {
   type=NiagaraDataInterfaceSkeletalMesh
   defaultValue {
    struct=NiagaraExt_VariableValue_DataInterface
    value {dataInterface=&self:NiagaraDataInterfaceSkeletalMesh_0}
   }
  }
 ]
}
