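# Niagara system that renders pale-blue cloud puffs, apparently for deploying the female Snowfall VinderTech glider chute.
# Two looping sprite emitters launch puffs in a wide cone with loop-age falloff, then apply drag and position-dependent local acceleration.
# A system Timeline float scales sprite size and spread velocity in both.
# One animates material parameters, the other uses SubUV animation and rotation; a component-renderer Ref emitter is disabled.
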NiagaraSystem /BRCosmetics/Gadgets/Assets/VinderTech_GliderChute/Glider_Snowfall_Female/FX/NS_snowfall_deploy_Cloud1
dependencies {
 usedDataInterfaces [&NiagaraDataInterfaceCurve, &NiagaraDataInterfaceArrayFloat]
 usedDynamicInputs []
 usedModules [&SystemState, &Timeline]
}
emitters {
 Initial_cloud_Bits {
  emitterUpdateScript [
   &EmitterState(Life Cycle Mode=::NewEnumerator1, Loop Behavior=::NewEnumerator1, Loop Duration=3, UseLoopDelay=true)
   &SpawnRate(SpawnRate=10, Use Spawn Probability=false) {module=SpawnRate001}
  ]
  particleSpawnScript [
   &InitializeParticle(
    Color Maximum=<rgba 0.4 0.68 0.8 1>
    Color Minimum=<rgba 0.3755 0.822 0.9387 1>
    Color Mode=::NewEnumerator2
    Lifetime Max=0.5
    Lifetime Min=2
    Lifetime Mode=::NewEnumerator1
    Mesh Scale Max=<0.75 0.7 0.7>
    Mesh Scale Min=<0.5 1 1>
    Mesh Scale Mode=::NewEnumerator2
    Position Offset=&RandomRangeVector(Minimum=<0 0 0>, Maximum=<0 100 0>)
    Sprite Rotation Angle=&RandomRangeFloat(Minimum=0, Maximum=0)
    Sprite Rotation Mode=::NewEnumerator0
    Sprite Size Mode=::NewEnumerator1
    Uniform Sprite Size Max=200
    Uniform Sprite Size Min=75
    UsePositionOffset=true
    Write Material Random=true
   )
   &AddVelocityInCone(
    Cone Angle=120
    Cone Axis=<0 -1 0>
    Module.Use Velocity Falloff On Cone Axis=false
    Velocity Distribution Along Cone Axis=0.15
    Velocity Strength=&Multiply_Float(
     A=&RandomRangeFloat(Minimum=250, Maximum=400)
     B=&FloatFromCurve(
      FloatCurve=DataInterface() {propertyValues {Curve {keys [{arriveTangent=-2.7353, interpMode=RCIM_Cubic, leaveTangent=-2.7353, tangentMode=RCTM_User, time=0, value=1}, <at 1 0 RCIM_Cubic>]}}}
      CurveIndex=$Emitter.NormalizedLoopAge
     )
    )
   )
   &AddVelocity(
    Velocity=&RandomRangeVector(Minimum=<-500 0 0>, Maximum=<500 0 0>)
    Velocity Speed Scale=&Multiply_Float(A=0.5, B=$System.Timeline.Float)
   )
   &ShapeLocation(Cylinder Radius=0, MaskSpawnGroup=false, Override Seed=false, Shape Primitive=::NewEnumerator1, Surface Only=false, UseOwnerScale=true)
  ]
  particleUpdateScript [
   &ParticleState()
   &Drag(Drag=3, Use Linear Drag=true, Use Rotational Drag=false)
   &AccelerationForce(
    Acceleration=&MakeCustomVectorFromBool(
     Bool=&self:is_pos_greater_than_0_local(
      B:NiagaraFloat=-50
      Default:Vector3f=&ConvertPositionToVector(Input Position=$Particles.Position, Convert To Absolute Large World Vector=false)
     )
     True Vector=<0 750 0>
     False Vector=<0 1000 0>
    )
    Coordinate Space=::Local
   )
   &SolveForcesAndVelocity(Clamp Velocity=false, Limit Acceleration=false)
   &DynamicMaterialParameters(
    Index 0 Param 1=&FloatFromCurve(FloatCurve=DataInterface() {propertyValues {Curve {keys [{tangentMode=RCTM_User, time=0, value=0}, {tangentMode=RCTM_User, time=1, value=1}]}}})
    Module.Param0WriteEnabled=true
    Module.Param1WriteEnabled=true
    Module.Param2WriteEnabled=true
    Module.Param3WriteEnabled=true
   )
   &ScaleColor(
    Scale Alpha=&Multiply_Float(
     A=&FloatFromCurve(FloatCurve=DataInterface() {propertyValues {Curve {keys [
          {arriveTangentWeight=0.3284, interpMode=RCIM_Cubic, leaveTangentWeight=0.3284, time=0, value=0}
          {arriveTangentWeight=0.2965, interpMode=RCIM_Cubic, leaveTangentWeight=0.2965, time=0.7, value=1}
          {arriveTangentWeight=0.3603, interpMode=RCIM_Cubic, leaveTangentWeight=0.3603, time=1, value=0}
         ]}}})
     B=1
    )
    ScaleA=true
    ScaleRGB=true
   )
   &ScaleSpriteSize(Scale Sprite Size Mode=::NewEnumerator0, Uniform Scale Factor=$System.Timeline.Float)
  ]
  rendererClasses [&NiagaraSpriteRendererProperties]
  renderers [
   {rendererClass=&NiagaraSpriteRendererProperties}
   {rendererClass=&NiagaraSpriteRendererProperties, rendererIndex=1}
  ]
 }
 Initial_cloud_Bits001 {
  emitterUpdateScript [
   &EmitterState(Life Cycle Mode=::NewEnumerator1, Loop Behavior=::NewEnumerator1, Loop Duration=3, UseLoopDelay=true)
   &SpawnRate(SpawnRate=20, Use Spawn Probability=false) {module=SpawnRate001}
  ]
  particleSpawnScript [
   &InitializeParticle(
    Color Maximum=<rgba 0.2362 0.3943 0.5906 0.7>
    Color Minimum=<rgba 0.4693 0.8414 0.9387 1>
    Color Mode=::NewEnumerator2
    Lifetime Max=0.5
    Lifetime Min=2
    Lifetime Mode=::NewEnumerator1
    Mesh Scale Max=<0.75 0.7 0.7>
    Mesh Scale Min=<0.5 1 1>
    Mesh Scale Mode=::NewEnumerator2
    Position Offset=&Multiply_VectorByFloat(Vector=&RandomRangeVector(Minimum=<0 0 0>, Maximum=<0 100 0>), Float=$System.Timeline.Float)
    Sprite Rotation Mode=::NewEnumerator2
    Sprite Size Mode=::NewEnumerator1
    Uniform Sprite Size Max=200
    Uniform Sprite Size Min=75
    UsePositionOffset=true
    Write Material Random=true
   )
   &AddVelocityInCone(
    Cone Angle=120
    Cone Axis=<0 -1 0>
    Module.Use Velocity Falloff On Cone Axis=false
    Velocity Distribution Along Cone Axis=0.15
    Velocity Strength=&Multiply_Float(
     A=&RandomRangeFloat(Minimum=250, Maximum=400)
     B=&FloatFromCurve(
      FloatCurve=DataInterface() {propertyValues {Curve {keys [{arriveTangent=-2.7353, interpMode=RCIM_Cubic, leaveTangent=-2.7353, tangentMode=RCTM_User, time=0, value=1}, <at 1 0 RCIM_Cubic>]}}}
      CurveIndex=$Emitter.NormalizedLoopAge
     )
    )
   )
   &AddVelocity(
    Velocity=&RandomRangeVector(Minimum=<-500 0 0>, Maximum=<500 0 0>)
    Velocity Speed Scale=&Multiply_Float(A=0.5, B=$System.Timeline.Float)
   )
   &ShapeLocation(
    Cylinder Height=&Multiply_Float(A=$System.Timeline.Float, B=100)
    Cylinder Radius=0
    MaskSpawnGroup=false
    Override Seed=false
    Shape Primitive=::NewEnumerator1
    Surface Only=false
    UseOwnerScale=true
   )
  ]
  particleUpdateScript [
   &ParticleState()
   &Drag(Drag=3, Use Linear Drag=true, Use Rotational Drag=false)
   &AccelerationForce(
    Acceleration=&MakeCustomVectorFromBool(
     Bool=&self:is_pos_greater_than_0_local(
      B:NiagaraFloat=-50
      Default:Vector3f=&ConvertPositionToVector(Input Position=$Particles.Position, Convert To Absolute Large World Vector=false)
     )
     True Vector=<0 750 0>
     False Vector=<0 1000 0>
    )
    Coordinate Space=::Local
   )
   &SolveForcesAndVelocity(Clamp Velocity=false, Limit Acceleration=false)
   &SubUVAnimation()
   &SpriteRotationRate(Rotation Rate=&RandomRangeFloat(Minimum=-25, Maximum=25))
   &ScaleColor(
    Scale Alpha=&Multiply_Float(
     A=&FloatFromCurve(FloatCurve=DataInterface() {propertyValues {Curve {keys [
          {arriveTangentWeight=0.3284, interpMode=RCIM_Cubic, leaveTangentWeight=0.3284, time=0, value=0}
          {arriveTangentWeight=0.2965, interpMode=RCIM_Cubic, leaveTangentWeight=0.2965, time=0.7, value=1}
          {arriveTangentWeight=0.3603, interpMode=RCIM_Cubic, leaveTangentWeight=0.3603, time=1, value=0}
         ]}}})
     B=1
    )
    ScaleA=true
    ScaleRGB=true
   )
   &ScaleSpriteSize(Scale Sprite Size Mode=::NewEnumerator0, Uniform Scale Factor=$System.Timeline.Float)
  ]
  rendererClasses [&NiagaraSpriteRendererProperties]
  renderers [{rendererClass=&NiagaraSpriteRendererProperties}]
 }
 Ref {
  bEnabled=false
  emitterUpdateScript [
   &EmitterState()
   &SpawnBurst_Instantaneous(Use Spawn Probability=false, UseLoopCountLimit=false)
  ]
  particleSpawnScript [&InitializeParticle(Color Mode=::NewEnumerator0, Lifetime=5, UsePositionOffset=false, Write Material Random=true)]
  particleUpdateScript [&ParticleState()]
  rendererClasses [&NiagaraComponentRendererProperties]
  renderers [{rendererClass=&NiagaraComponentRendererProperties, rendererIndex=0}]
 }
}
summary {
 systemName=NS_snowfall_deploy_Cloud1
 emitters [
  {bEnabled=false, emitterName=Ref, rendererClasses [&NiagaraComponentRendererProperties]}
  {emitterName=Initial_cloud_Bits001, rendererClasses [&NiagaraSpriteRendererProperties]}
  {emitterName=Initial_cloud_Bits, rendererClasses [&NiagaraSpriteRendererProperties]}
 ]
 userVariables [
  User.bIsFrontEnd {type=NiagaraBool}
  User.bIsFrontEndPreview {type=NiagaraBool}
  User.bIsInWater {type=NiagaraBool}
 ]
}
